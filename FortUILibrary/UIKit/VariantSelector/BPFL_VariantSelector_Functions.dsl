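; Blueprint function library for the Fortnite variant selector UI that works out which warning text applies to a cosmetic item or variant.
; GetConflictsFromItem maps age gate, access restriction and lock states to texts by conflict type.
; GetConflictsFromVariant builds on that map, adding variant age gate, lock and conflict texts.
; GetPrioritizedConflictFromMap returns the first non-empty text and its type in fixed priority order, item issues before variant issues.
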(fn GetConflictsFromItem
  (params (SelectedItem :type Fort%20Item%20VM%20Object%20Reference) (VariantSelectorVM :type Fort%20Variant%20Selector%20VM%20Object%20Reference))
  (bind _selecteditem (Variables|GetSelectedItem))
  (bind _itemdetails (Variables|/Script/FortniteUI.FortItemVM|GetItemDetails _selecteditem))
  (bind _selecteditem_2 (Variables|GetSelectedItem))
  (bind _selecteditem_3 (Variables|GetSelectedItem))
  (bind _itemdetails_2 (Variables|/Script/FortniteUI.FortItemVM|GetItemDetails _selecteditem_3))
  (Utilities|IsValid SelectedItem
    (:Is%20Valid
      (Utilities|IsValid VariantSelectorVM
        (:Is%20Valid
          (Utilities|FlowControl|Sequence
            (:then_0
              ;[0 0 0] Item is age gated
              (if (and (Variables|/Script/FortniteUI.FortVariantSelectorVM|GetbShowAgeGateWarnings (Variables|GetVariantSelectorVM)) (Variables|/Script/FortniteUI.FortItemDetailsVM|GetbIsAgeGated _itemdetails))
                (Utilities|Map|Add (LocalVariables|GetConflictMap) "NewEnumerator4" (Variables|/Script/FortniteUI.FortItemDetailsVM|GetAccessRestrictionDescription _itemdetails))))
            (:then_1
              ;[0 0 0] Item has an access restriction description
              (if (Variables|/Script/FortniteUI.FortItemDetailsVM|GetbHasAccessRestriction _itemdetails_2)
                (Utilities|Map|Add (LocalVariables|GetConflictMap) "NewEnumerator3" (Variables|/Script/FortniteUI.FortItemDetailsVM|GetAccessRestrictionDescription _itemdetails_2))))
            (:then_2
              ;[0 0 0] Item is locked
              (if (Variables|/Script/FortniteUI.FortItemVM|GetbIsLocked _selecteditem_2)
                (Utilities|Map|Add (LocalVariables|GetConflictMap) "NewEnumerator5" (Variables|/Script/FortniteUI.FortItemDetailsVM|GetUnlockRequirements (Variables|/Script/FortniteUI.FortItemVM|GetItemDetails _selecteditem_2)))))
            (:then_3
              _)))
        (:Is%20Not%20Valid
          _))
      (return (LocalVariables|GetConflictMap)))
    (:Is%20Not%20Valid
      _)))

(fn GetConflictsFromVariant
  (params (SelectedVariant :type Fort%20Variant%20Item%20VM%20Object%20Reference) (VariantSelectorVM :type Fort%20Variant%20Selector%20VM%20Object%20Reference))
  (bind _selectedvariant (Variables|GetSelectedVariant))
  (bind _selectedvariant_2 (Variables|GetSelectedVariant))
  (bind _selectedvariant_3 (Variables|GetSelectedVariant))
  (Utilities|IsValid SelectedVariant
    (:Is%20Valid
      (Utilities|IsValid VariantSelectorVM
        (:Is%20Valid
          (Utilities|FlowControl|Sequence
            (:then_0
              ;[0 0 0] Get conflicts from item
              (LocalVariables|SetConflictMap (Default|GetConflictsfromItem (Variables|/Script/FortniteUI.FortVariantItemVM|GetCosmeticItem (Variables|GetSelectedVariant)) (Variables|GetVariantSelectorVM))))
            (:then_1
              ;[0 0 0] Variant is age gated
              (if (and (Variables|/Script/FortniteUI.FortVariantSelectorVM|GetbShowAgeGateWarnings (Variables|GetVariantSelectorVM)) (Variables|/Script/FortniteUI.FortVariantItemVM|GetbHasAgeGate _selectedvariant_3))
                (Utilities|Map|Add (LocalVariables|GetConflictMap) "NewEnumerator1" (Variables|/Script/FortniteUI.FortVariantItemVM|GetAgeGateText _selectedvariant_3))))
            (:then_2
              ;[0 0 0] Variant is locked (e.g., unlock variants through quests)
              (if (Variables|/Script/FortniteUI.FortVariantItemVM|GetbIsLocked _selectedvariant_2)
                (Utilities|Map|Add (LocalVariables|GetConflictMap) "NewEnumerator2" (Variables|/Script/FortniteUI.FortVariantItemVM|GetUnlockRequirements _selectedvariant_2))))
            (:then_3
              ;[0 0 0] Variant has a conflict (e.g., color conflicts)
              (if (Variables|/Script/FortniteUI.FortVariantItemVM|GetbHasConflict _selectedvariant)
                (Utilities|Map|Add (LocalVariables|GetConflictMap) "NewEnumerator0" (Variables|/Script/FortniteUI.FortVariantItemVM|GetConflictingText _selectedvariant))))
            (:then_4
              _)))
        (:Is%20Not%20Valid
          _))
      (return (LocalVariables|GetConflictMap)))
    (:Is%20Not%20Valid
      _)))

(fn GetPrioritizedConflictFromMap
  (params (Conflicts :type Map%20of%20E_VariantConflictType%20Enums%20to%20Texts :default "NewEnumerator6"))
  (bind _conflicts (Variables|GetConflicts))
  (for _array_element (Utilities|Array|MakeArray "NewEnumerator4" "NewEnumerator3" "NewEnumerator5" "NewEnumerator1" "NewEnumerator2" "NewEnumerator0")
    (if (and (.ReturnValue (Utilities|Map|Find _conflicts _array_element)) (not (Utilities|Text|TextIsEmpty (Utilities|Map|Find _conflicts _array_element))))
      (return (Utilities|Map|Find _conflicts _array_element) _array_element)))
  (return "" "NewEnumerator6"))
