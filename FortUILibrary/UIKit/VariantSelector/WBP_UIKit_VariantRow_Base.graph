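# Base UserWidget for a row in the variant selector UI, with header text, a divider, a content named slot and a disclaimer line.
# InitializeRow, GetDesiredFocusTarget and OnConflictsUpdated are empty stubs, and dispatchers are declared for item preview, item engage and conflict updates.
# UseCustomizeMenuStyle toggles between a single-line header with divider and a two-line category and item header.
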WidgetBlueprint /FortUILibrary/UIKit/VariantSelector/WBP_UIKit_VariantRow_Base
graphs [
 EventGraph {links [], nodes []}
 GetDesiredFocusTarget {nodes [
   GetDesiredFocusTarget() {is=title K2Node_FunctionEntry}
   Return Node!(execute, FocusTarget:Widget Object Reference)
  ]}
 InitializeRow {nodes [InitializeRow!() -> (
    then
    ChannelVM:Fort Variant Channel VM Object Reference
    VariantSelectorVM:Fort Variant Selector VM Object Reference
    AdditionalArguments:Map of Names to Booleans
    TargetStyleTag:Gameplay Tag Structure
   ) {is=title K2Node_FunctionEntry}]}
 OnConflictsUpdated {nodes [OnConflictsUpdated!() -> (then, Conflicts:Map of E_VariantConflictType Enums to Texts=NewEnumerator6) {is=title K2Node_FunctionEntry}]}
 OnVariantItemEngaged {nodes [OnVariantItemEngaged!() -> (then, VariantItem:Fort Variant Item VM Object Reference, VariantChannel:Fort Variant Channel VM Object Reference, AdditionalInfo:String) {is=title K2Node_FunctionEntry}]}
 OnVariantItemPreviewed {nodes [OnVariantItemPreviewed!() -> (then, VariantItem:Fort Variant Item VM Object Reference) {is=title K2Node_FunctionEntry}]}
 UseCustomizeMenuStyle {nodes [
   n1 = UseCustomizeMenuStyle!() -> (then, UseCustomizeMenuStyle:bool=~) {is=title K2Node_FunctionEntry}
   > n2 = Branch(Condition=@n1)
   > SetVisibility(self=Get Image_HeaderDivider() -> (Image_HeaderDivider:Image Object Reference), InVisibility=Collapsed)
   > SetVisibility(self=Get SingleLineHeader_VisBox() -> (SingleLineHeader_VisBox:Vertical Box Object Reference), InVisibility=Collapsed)
   > SetVisibility(self=Get DoubleLinedHeader_VisBox() -> (DoubleLinedHeader_VisBox:Vertical Box Object Reference), InVisibility=Visible)
   SetVisibility(execute=@n2.else, self=Get Image_HeaderDivider(), InVisibility=Visible)
   > SetVisibility(self=Get SingleLineHeader_VisBox(), InVisibility=Visible)
   > SetVisibility(self=Get DoubleLinedHeader_VisBox(), InVisibility=Collapsed)
  ]}
]
variables [
 OnConflictsUpdated {}
 OnVariantItemEngaged {}
 OnVariantItemPreviewed {}
]
widgets {
 info {inheritedWidgetCount=0, parentClass=&UserWidget, rootWidgetClass=&VerticalBox, widgetCount=9}
 widgets [
  VerticalBox_26 {widgetClassPath=&VerticalBox, widget=VerticalBox_26}
  SingleLineHeader_VisBox {
   widgetClassPath=&VerticalBox
   bIsVariable=true
   parent=VerticalBox_26
   widget=SingleLineHeader_VisBox
   slot {class=VerticalBoxSlot, properties {padding=<ltrb 0 0 0 8>}}
  }
  Text_RowCategoryAndItem {
   widgetClassPath=&CommonTextBlock
   bIsVariable=true
   parent=SingleLineHeader_VisBox
   widget=Text_RowCategoryAndItem
   properties {
    bAutoCollapseWithEmptyText=true
    mobileFontSizeMultiplier=1.4
    scrollStyle=&TextScrollStyle-Base_C
    text="[PH] Row Title"
    textTransformPolicy=ToUpper
    font {
     bForceMonospaced=false
     bMaterialIsStencil=false
     fontMaterial=null
     fontObject=&HeadingNow
     letterSpacing=0
     monospacedWidth=1
     size=18
     skewAmount=0
     typefaceFontName=75 Medium
     outlineSettings {
      bApplyOutlineToDropShadows=false
      bMiteredCorners=false
      bSeparateFillAlpha=false
      outlineMaterial=null
      outlineSize=0
      outlineColor=<rgba 0 0 0 1>
     }
    }
    shadowOffset=<0 0>
    strikeBrush {
     bIsDynamicallyLoaded=false
     drawAs=NoDrawType
     imageType=NoImage
     mirroring=NoMirror
     resourceName=None
     resourceObject=null
     tiling=NoTile
     imageSize=<32 32>
     margin=<ltrb 0 0 0 0>
     outlineSettings {
      bUseBrushTransparency=false
      roundingType=HalfHeightRadius
      width=0
      color {colorUseRule=UseColor_Specified, specifiedColor=<rgba 0 0 0 0>}
      cornerRadii=<0 0 0 0>
     }
     tintColor {colorUseRule=UseColor_Specified, specifiedColor=<rgba 1 1 1 1>}
     uVRegion {bIsValid=false, max=<0 0>, min=<0 0>}
    }
   }
   slot {class=VerticalBoxSlot}
  }
  DoubleLinedHeader_VisBox {
   widgetClassPath=&VerticalBox
   bIsVariable=true
   parent=VerticalBox_26
   widget=DoubleLinedHeader_VisBox
   properties {visibility=Collapsed}
   slot {class=VerticalBoxSlot, properties {padding=<ltrb 16 0 0 32>}}
  }
  Text_RowCategory {
   widgetClassPath=&CommonTextBlock
   bIsVariable=true
   parent=DoubleLinedHeader_VisBox
   widget=Text_RowCategory
   properties {
    bAutoCollapseWithEmptyText=true
    mobileFontSizeMultiplier=1.4
    scrollStyle=&TextScrollStyle-Base_C
    text="[PH] Row Category"
    textTransformPolicy=ToUpper
    font {
     bForceMonospaced=false
     bMaterialIsStencil=false
     fontMaterial=null
     fontObject=&HeadingNow
     letterSpacing=0
     monospacedWidth=1
     size=18
     skewAmount=0
     typefaceFontName=75 Medium
     outlineSettings {
      bApplyOutlineToDropShadows=false
      bMiteredCorners=false
      bSeparateFillAlpha=false
      outlineMaterial=null
      outlineSize=0
      outlineColor=<rgba 0 0 0 1>
     }
    }
    shadowOffset=<0 0>
    strikeBrush {
     bIsDynamicallyLoaded=false
     drawAs=NoDrawType
     imageType=NoImage
     mirroring=NoMirror
     resourceName=None
     resourceObject=null
     tiling=NoTile
     imageSize=<32 32>
     margin=<ltrb 0 0 0 0>
     outlineSettings {
      bUseBrushTransparency=false
      roundingType=HalfHeightRadius
      width=0
      color {colorUseRule=UseColor_Specified, specifiedColor=<rgba 0 0 0 0>}
      cornerRadii=<0 0 0 0>
     }
     tintColor {colorUseRule=UseColor_Specified, specifiedColor=<rgba 1 1 1 1>}
     uVRegion {bIsValid=false, max=<0 0>, min=<0 0>}
    }
   }
   slot {class=VerticalBoxSlot, properties {verticalAlignment=VAlign_Top}}
  }
  Text_RowItem {
   widgetClassPath=&CommonTextBlock
   bIsVariable=true
   parent=DoubleLinedHeader_VisBox
   widget=Text_RowItem
   properties {
    applyLineHeightToBottomLine=false
    bAutoCollapseWithEmptyText=true
    clipping=OnDemand
    mobileFontSizeMultiplier=1.2
    pixelSnapping=Disabled
    scrollStyle=/FortUILibrary/Styles/Typography/Foundation/TextScrollStyle-NoFade.TextScrollStyle-NoFade_C
    textOverflowPolicy=Ellipsis
    colorAndOpacity {colorUseRule=UseColor_Specified, specifiedColor=<rgba 1 1 1 0.7>}
    font {
     bForceMonospaced=false
     bMaterialIsStencil=false
     fontMaterial=null
     fontObject=&HeadingNow
     letterSpacing=0
     monospacedWidth=1
     size=14.25
     skewAmount=0
     typefaceFontName=74 Regular
     outlineSettings {
      bApplyOutlineToDropShadows=false
      bMiteredCorners=false
      bSeparateFillAlpha=false
      outlineMaterial=null
      outlineSize=0
      outlineColor=<rgba 0 0 0 1>
     }
    }
    shadowOffset=<0 0>
   }
   slot {class=VerticalBoxSlot, properties {verticalAlignment=VAlign_Top, padding=<ltrb 0 4 0 0>}}
  }
  Image_HeaderDivider {
   widgetClassPath=&Image
   bIsVariable=true
   parent=VerticalBox_26
   widget=Image_HeaderDivider
   properties {brush {
     bIsDynamicallyLoaded=false
     drawAs=Image
     imageType=NoImage
     mirroring=NoMirror
     resourceName=None
     resourceObject=/FortUILibrary/MaterialInstances/MI_UI_Customization_Separator
     tiling=NoTile
     imageSize=<16 2>
     margin=<ltrb 0 0 0 0>
     outlineSettings {
      bUseBrushTransparency=false
      roundingType=HalfHeightRadius
      width=0
      color {colorUseRule=UseColor_Specified, specifiedColor=<rgba 0 0 0 0>}
      cornerRadii=<0 0 0 0>
     }
     tintColor {colorUseRule=UseColor_Specified, specifiedColor=<rgba 1 1 1 1>}
     uVRegion {bIsValid=false, max=<0 0>, min=<0 0>}
    }}
   slot {class=VerticalBoxSlot, properties {verticalAlignment=VAlign_Top, padding=<ltrb 0 0 0 16>}}
  }
  Slot_Content {widgetClassPath=&NamedSlot, bIsVariable=true, parent=VerticalBox_26, widget=Slot_Content, slot {class=VerticalBoxSlot}}
  Text_Disclaimer {
   widgetClassPath=&CommonTextBlock
   bIsVariable=true
   parent=VerticalBox_26
   widget=Text_Disclaimer
   properties {
    bAutoCollapseWithEmptyText=true
    justification=Right
    mobileFontSizeMultiplier=1.4
    scrollStyle=&TextScrollStyle-Base_C
    text=Disclaimer Text
    textTransformPolicy=ToUpper
    font {
     bForceMonospaced=false
     bMaterialIsStencil=false
     fontMaterial=null
     fontObject=&HeadingNow
     letterSpacing=0
     monospacedWidth=1
     size=18
     skewAmount=0
     typefaceFontName=74 Regular
     outlineSettings {
      bApplyOutlineToDropShadows=false
      bMiteredCorners=false
      bSeparateFillAlpha=false
      outlineMaterial=null
      outlineSize=0
      outlineColor=<rgba 0 0 0 1>
     }
    }
    shadowOffset=<0 0>
    strikeBrush {
     bIsDynamicallyLoaded=false
     drawAs=NoDrawType
     imageType=NoImage
     mirroring=NoMirror
     resourceName=None
     resourceObject=null
     tiling=NoTile
     imageSize=<32 32>
     margin=<ltrb 0 0 0 0>
     outlineSettings {
      bUseBrushTransparency=false
      roundingType=HalfHeightRadius
      width=0
      color {colorUseRule=UseColor_Specified, specifiedColor=<rgba 0 0 0 0>}
      cornerRadii=<0 0 0 0>
     }
     tintColor {colorUseRule=UseColor_Specified, specifiedColor=<rgba 1 1 1 1>}
     uVRegion {bIsValid=false, max=<0 0>, min=<0 0>}
    }
   }
   slot {class=VerticalBoxSlot, properties {padding=<ltrb 0 10 0 0>}}
  }
 ]
}
