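# Variant selector row widget for stepping a numeric channel value with single and double arrow buttons (±1, ±10).
# InitializeRow caches the numeric channel view model, which each step changes with wraparound before broadcasting On Variant Item Engaged.
# Holding re-arms a one-shot timer whose interval comes from a curve over accumulated hold time, so repeats accelerate.
# Release applies one step if no repeat fired, updates the item VM, clears the timer and broadcasts again.
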WidgetBlueprint /FortUILibrary/UIKit/VariantSelector/WBP_UIKit_VariantRow_Numeric
graphs [
 EventGraph {nodes [
   n1 = "Hold Action Ended (Button_TenDigitDown)"() {is=title K2Node_ComponentBoundEvent}
   n2 = Branch(execute=[@n1.then, @n12.then, @n13.then, @n15.then], Condition=Get InputHasBeenProcessed() -> (InputHasBeenProcessed:bool))
   UpdateItemVM(
    execute=[@n5.then, @n2.then]
    self:Fort Variant Numeric Channel VM Object Reference=Get FortVariantNumericChannelVM() -> (FortVariantNumericChannelVM:Fort Variant Numeric Channel VM Object Reference)
   )
   > Clear and Invalidate Timer by Handle(Handle=Get HoldTimer() -> (HoldTimer:Timer Handle Structure))
   > Set InputHasBeenProcessed(InputHasBeenProcessed:bool=true) -> (Output_Get:bool)
   n3 = Get Text_NumericalValue() -> (Text_NumericalValue:Common Text Object Reference)
   n4 = Get FortVariantNumericChannelVM()
   > Call On Variant Item Engaged!(
    execute:Exec
    self:WBP UIKit Variant Row Base Object Reference
    VariantItem:Fort Variant Item VM Object Reference=Get EquippedVariantItem!(self:Fort Variant Channel VM Object Reference=@n4) -> (EquippedVariantItem:Fort Variant Item VM Object Reference)
    VariantChannel:Fort Variant Channel VM Object Reference=@n4
    AdditionalInfo:String="To String (Text)"(InText=Get Text(self:Text Object Reference=@n3))
   ) -> (then:Exec) {is=title K2Node_CallDelegate}
   n5 = Update Value!(execute=@n2.else, self, Delta:int=Get IncrementDelta() -> (IncrementDelta:int=0)) -> (then)
   "Hold Action Started (Button_ZeroDigitUp)"() {is=title K2Node_ComponentBoundEvent}
   > n6 = Set IncrementDelta(IncrementDelta:int=1) -> (Output_Get:int=0)
   Set InputHasBeenProcessed(execute=[@n6.then, @n14.then, @n11.then, @n10.then], InputHasBeenProcessed=false)
   > n7 = Set ElapsedTime(ElapsedTime:double=0) -> (Output_Get:double)
   n8 = On Input Hold Timer Event() {is=title K2Node_CustomEvent}
   n9 = Set Timer by Event(
    execute=[@n20.then, @n7.then]
    Delegate=@n8
    Time=GetRuntimeFloatCurveValue!(
     Curve:"Runtime Float Curve Structure (by ref)"=Get HoldInputCurve() -> (HoldInputCurve:Runtime Float Curve Structure)
     InTime:float=Get ElapsedTime() -> (ElapsedTime:double)
     InDefaultValue:float=0
    ) -> (ReturnValue:float)
    bLooping=false
   )
   > Set HoldTimer(HoldTimer:Timer Handle Structure=@n9) -> (Output_Get:Timer Handle Structure)
   "Hold Action Started (Button_TenDigitUp)"() {is=title K2Node_ComponentBoundEvent}
   > n10 = Set IncrementDelta(IncrementDelta=10) -> (Output_Get=0)
   "Hold Action Started (Button_TenDigitDown)"() {is=title K2Node_ComponentBoundEvent}
   > n11 = Set IncrementDelta(IncrementDelta=-10) -> (Output_Get=0)
   n12 = "Hold Action Ended (Button_ZeroDigitDown)"() {is=title K2Node_ComponentBoundEvent}
   n13 = "Hold Action Ended (Button_ZeroDigitUp)"() {is=title K2Node_ComponentBoundEvent}
   "Hold Action Started (Button_ZeroDigitDown)"() {is=title K2Node_ComponentBoundEvent}
   > n14 = Set IncrementDelta(IncrementDelta=-1) -> (Output_Get=0)
   n15 = "Hold Action Ended (Button_TenDigitUp)"() {is=title K2Node_ComponentBoundEvent}
   n16 = Update Value!() -> (OutputDelegate, then, Delta:int) {is=title K2Node_CustomEvent}
   n17 = Get FortVariantNumericChannelVM()
   > ChangeNumericalValue!(execute, self:Fort Variant Numeric Channel VM Object Reference=@n17, NumericalDelta:int=@n16.Delta, bAllowWrap:bool=true) -> (then)
   n18 = Get Text_NumericalValue()
   > Call On Variant Item Engaged(VariantItem=Get EquippedVariantItem(self=@n17), VariantChannel=@n17, AdditionalInfo="To String (Text)"(InText=Get Text(self=@n18)))
   n19 = Get Timer Elapsed Time by Handle!(Handle:Timer Handle Structure=Get HoldTimer()) -> (ReturnValue:float) {is=title K2Node_CallFunction}
   n20 = Update Value(execute=@n22.then, Delta=Get IncrementDelta() -> (IncrementDelta=0)) {is=title K2Node_CallFunction}
   n21 = float + float(A=Get ElapsedTime(), B=@n19)
   Set InputHasBeenProcessed(execute=@n8.then, InputHasBeenProcessed=true)
   > n22 = Set ElapsedTime(ElapsedTime=@n21)
  ]}
 GetDesiredFocusTarget {nodes [
   GetDesiredFocusTarget() {is=title K2Node_FunctionEntry}
   > Return Node!(execute, FocusTarget:Widget Object Reference=Get Button_ZeroDigitDown() -> (Button_ZeroDigitDown:Panel Button Object Reference))
  ]}
 InitializeRow {nodes [
   n1 = InitializeRow!() -> (
    then
    ChannelVM:Fort Variant Channel VM Object Reference
    VariantSelectorVM:Fort Variant Selector VM Object Reference
    AdditionalArguments:Map of Names to Booleans
    TargetStyleTag:Gameplay Tag Structure
   ) {is=title K2Node_FunctionEntry}
   > n2 = Cast To FortVariantNumericChannelVM!(execute, Object=@n1) -> (then, CastFailed, AsFort Variant Numeric Channel VM:Fort Variant Numeric Channel VM Object Reference) {is=title K2Node_DynamicCast}
   > Set FortVariantNumericChannelVM(FortVariantNumericChannelVM:Fort Variant Numeric Channel VM Object Reference=@n2) -> (Output_Get:Fort Variant Numeric Channel VM Object Reference)
  ]}
]
variables [
 ElapsedTime {type=number}
 HoldInputCurve {
  type=RuntimeFloatCurve
  default {
   externalCurve=null
   editorCurveData {
    defaultValue=3.4028234663852886e+38
    postInfinityExtrap=RCCE_Constant
    preInfinityExtrap=RCCE_Constant
    keys [
     {
      arriveTangent=0
      arriveTangentWeight=0
      interpMode=RCIM_Constant
      leaveTangent=0
      leaveTangentWeight=0
      tangentMode=RCTM_Auto
      tangentWeightMode=RCTWM_WeightedNone
      time=0
      value=0.1
     }
     {
      arriveTangent=0
      arriveTangentWeight=0
      interpMode=RCIM_Constant
      leaveTangent=0
      leaveTangentWeight=0
      tangentMode=RCTM_Auto
      tangentWeightMode=RCTWM_WeightedNone
      time=2
      value=0.05
     }
     {
      arriveTangent=0
      arriveTangentWeight=0
      interpMode=RCIM_Constant
      leaveTangent=0
      leaveTangentWeight=0
      tangentMode=RCTM_Auto
      tangentWeightMode=RCTWM_WeightedNone
      time=4
      value=0.025
     }
    ]
   }
  }
 }
 HoldTimer {type=TimerHandle}
 IncrementDelta {type=integer}
 InputHasBeenProcessed {type=boolean}
]
widgets {
 info {
  inheritedWidgetCount=0
  namedSlotCount=1
  parentClass=&/WBP_UIKit_VariantRow_Base.WBP_UIKit_VariantRow_Base_C
  rootWidgetClass=null
  widgetCount=10
 }
 widgets [
  HorizontalBox_0 {widgetClassPath=&HorizontalBox, widget=HorizontalBox_0}
  Button_TenDigitDown {
   widgetClassPath=/FortUILibrary/Foundation/Buttons/PanelButton
   bIsVariable=true
   parent=HorizontalBox_0
   widget=Button_TenDigitDown
   properties {
    inputActionHorizontalAlignment=HAlign_Right
    inputActionVerticalAlignment=VAlign_Center
    minHeight=88
    minWidth=88
    onButtonBaseClicked=()
    onButtonBaseDoubleClicked=()
    onButtonBaseHovered=()
    onButtonBaseUnhovered=()
    onSelectedChangedBase=()
    style=/FortUILibrary/Foundation/Buttons/ButtonStyles/ButtonStyle_CycleArrow_Left_Double.ButtonStyle_CycleArrow_Left_Double_C
    visibility=Visible
    inputActionRenderTranslation=<22 0>
    inputActionRimBrushSize=<40 40>
   }
   slot {class=HorizontalBoxSlot, properties {verticalAlignment=VAlign_Center, padding=<ltrb 0 0 16 0>}}
  }
  Button_ZeroDigitDown {
   widgetClassPath=&PanelButton
   bIsVariable=true
   parent=HorizontalBox_0
   widget=Button_ZeroDigitDown
   properties {
    inputActionHorizontalAlignment=HAlign_Right
    inputActionVerticalAlignment=VAlign_Center
    minHeight=88
    minWidth=88
    onButtonBaseClicked=()
    onButtonBaseDoubleClicked=()
    onButtonBaseHovered=()
    onButtonBaseUnhovered=()
    onSelectedChangedBase=()
    style=/FortUILibrary/Foundation/Buttons/ButtonStyles/ButtonStyle_CycleArrow_Left.ButtonStyle_CycleArrow_Left_C
    visibility=Visible
    inputActionRenderTranslation=<22 0>
    inputActionRimBrushSize=<40 40>
   }
   slot {class=HorizontalBoxSlot, properties {verticalAlignment=VAlign_Center}}
  }
  Border_Value {
   widgetClassPath=&Border
   parent=HorizontalBox_0
   widget=Border_Value
   properties {
    background {
     bIsDynamicallyLoaded=false
     drawAs=Image
     imageType=FullColor
     mirroring=NoMirror
     resourceName=None
     resourceObject=/FortUILibrary/Styles/Border/Materials/M_Border_Solid_DkBlue
     tiling=NoTile
     imageSize=<32 32>
     margin=<ltrb 0 0 0 0>
     outlineSettings {
      bUseBrushTransparency=false
      roundingType=FixedRadius
      width=4
      color {colorUseRule=UseColor_Specified, specifiedColor=<rgba 1 1 1 0.2>}
      cornerRadii=<0 0 0 0>
     }
     tintColor {colorUseRule=UseColor_Specified, specifiedColor=<rgba 1 1 1 1>}
     uVRegion {bIsValid=false, max=<0 0>, min=<0 0>}
    }
    brushColor=<rgba 1 1 1 0.2>
    padding=<ltrb 0 0 0 0>
   }
   slot {class=HorizontalBoxSlot}
  }
  Overlay_74 {widgetClassPath=&Overlay, parent=Border_Value, widget=Overlay_74, slot {class=BorderSlot, properties {padding=<ltrb 0 0 0 0>}}}
  Image_TopBorder {
   widgetClassPath=&Image
   parent=Overlay_74
   widget=Image_TopBorder
   properties {
    brush {
     bIsDynamicallyLoaded=false
     drawAs=Image
     imageType=NoImage
     mirroring=NoMirror
     resourceName=None
     resourceObject=null
     tiling=NoTile
     imageSize=<192 4>
     margin=<ltrb 0 0 0 0>
     outlineSettings {
      bUseBrushTransparency=false
      roundingType=HalfHeightRadius
      width=0
      color {colorUseRule=UseColor_Specified, specifiedColor=<rgba 0 0 0 0>}
      cornerRadii=<0 0 0 0>
     }
     tintColor {colorUseRule=UseColor_Specified, specifiedColor=<rgba 1 1 1 1>}
     uVRegion {bIsValid=false, max=<0 0>, min=<0 0>}
    }
    colorAndOpacity=<rgba 1 1 1 0.2>
   }
   slot {class=OverlaySlot, properties {horizontalAlignment=HAlign_Fill}}
  }
  Text_NumericalValue {
   widgetClassPath=&CommonTextBlock
   parent=Overlay_74
   widget=Text_NumericalValue
   properties {
    bIsScrollingEnabled=false
    justification=Center
    text=999
    colorAndOpacity {colorUseRule=UseColor_Specified, specifiedColor=<rgba 1 1 1 0.7>}
    font {
     bForceMonospaced=false
     bMaterialIsStencil=false
     fontMaterial=null
     fontObject=&HeadingNow
     letterSpacing=0
     monospacedWidth=1
     size=36
     skewAmount=0
     typefaceFontName=74 Regular
     outlineSettings {
      bApplyOutlineToDropShadows=false
      bMiteredCorners=false
      bSeparateFillAlpha=false
      outlineMaterial=null
      outlineSize=0
      outlineColor=<rgba 0 0 0 1>
     }
    }
    shadowOffset=<0 0>
   }
   slot {class=OverlaySlot, properties {horizontalAlignment=HAlign_Center, verticalAlignment=VAlign_Center, padding=<ltrb 16 -16 16 -16>}}
  }
  Image_BottomBorder {
   widgetClassPath=&Image
   parent=Overlay_74
   widget=Image_BottomBorder
   properties {
    brush {
     bIsDynamicallyLoaded=false
     drawAs=Image
     imageType=NoImage
     mirroring=NoMirror
     resourceName=None
     resourceObject=null
     tiling=NoTile
     imageSize=<192 4>
     margin=<ltrb 0 0 0 0>
     outlineSettings {
      bUseBrushTransparency=false
      roundingType=HalfHeightRadius
      width=0
      color {colorUseRule=UseColor_Specified, specifiedColor=<rgba 0 0 0 0>}
      cornerRadii=<0 0 0 0>
     }
     tintColor {colorUseRule=UseColor_Specified, specifiedColor=<rgba 1 1 1 1>}
     uVRegion {bIsValid=false, max=<0 0>, min=<0 0>}
    }
    colorAndOpacity=<rgba 1 1 1 0.2>
   }
   slot {class=OverlaySlot, properties {horizontalAlignment=HAlign_Fill, verticalAlignment=VAlign_Bottom}}
  }
  Button_ZeroDigitUp {
   widgetClassPath=&PanelButton
   bIsVariable=true
   parent=HorizontalBox_0
   widget=Button_ZeroDigitUp
   properties {
    inputActionHorizontalAlignment=HAlign_Left
    inputActionVerticalAlignment=VAlign_Center
    minHeight=88
    minWidth=88
    onButtonBaseClicked=()
    onButtonBaseDoubleClicked=()
    onButtonBaseHovered=()
    onButtonBaseUnhovered=()
    onSelectedChangedBase=()
    style=/FortUILibrary/Foundation/Buttons/ButtonStyles/ButtonStyle_CycleArrow_Right.ButtonStyle_CycleArrow_Right_C
    visibility=Visible
    inputActionRenderTranslation=<-22 0>
    inputActionRimBrushSize=<40 40>
   }
   slot {class=HorizontalBoxSlot, properties {verticalAlignment=VAlign_Center}}
  }
  Button_TenDigitUp {
   widgetClassPath=&PanelButton
   bIsVariable=true
   parent=HorizontalBox_0
   widget=Button_TenDigitUp
   properties {
    inputActionHorizontalAlignment=HAlign_Left
    inputActionVerticalAlignment=VAlign_Center
    minHeight=88
    minWidth=88
    onButtonBaseClicked=()
    onButtonBaseDoubleClicked=()
    onButtonBaseHovered=()
    onButtonBaseUnhovered=()
    onSelectedChangedBase=()
    style=/FortUILibrary/Foundation/Buttons/ButtonStyles/ButtonStyle_CycleArrow_Right_Double.ButtonStyle_CycleArrow_Right_Double_C
    visibility=Visible
    inputActionRenderTranslation=<-22 0>
    inputActionRimBrushSize=<40 40>
   }
   slot {class=HorizontalBoxSlot, properties {verticalAlignment=VAlign_Center, padding=<ltrb 16 0 0 0>}}
  }
 ]
}
