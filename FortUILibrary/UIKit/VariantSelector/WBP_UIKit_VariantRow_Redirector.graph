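# Variant-selector row widget, built on FortRedirectorVariantWrapper, that redirects engaged variant items to their matching channel view models.
# InitializeRow maps each channel VM by its channel tag and registers it with the parent wrapper.
# On engagement it looks up the channel by tag, forwards the item, then broadcasts OnVariantItemEngaged with the CustomData text after the colon.
# Desired focus comes from the Container's first child via BPI_DynVariantRow_FocusTargetWidget, falling back to the Container.
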WidgetBlueprint /FortUILibrary/UIKit/VariantSelector/WBP_UIKit_VariantRow_Redirector
graphs [
 EventGraph {nodes [
   n1 = Event HandleVariantEngaged!() -> (OutputDelegate, then, InVariantChannelInfo:"Mcp Variant Channel Info Structure (by ref)", ItemVM:Fort Variant Item VM Object Reference) {is=title K2Node_Event}
   > Is Valid(InputObject=@n1.ItemVM)
   n2 = Break Mcp Variant Channel Info!(McpVariantChannelInfo:"Mcp Variant Channel Info Structure (by ref)"=@n1.InVariantChannelInfo) -> (
    OwnedVariantTags:Gameplay Tag Container Structure
    ItemVariantIsUsedFor:Item Definition Base Object Reference
    CustomData:String
    VariantChannelTag:Gameplay Tag Structure
    ActiveVariantTag:Gameplay Tag Structure
   ) {is=title K2Node_BreakStruct}
   n3 = Find(
    TargetMap:Map of Gameplay Tag Structures to Fort Variant Channel VM Object References=Get VariantChannelsMap() -> (VariantChannelsMap:Map of Gameplay Tag Structures to Fort Variant Channel VM Object References)
    Key:"Gameplay Tag Structure (by ref)"=@n2.VariantChannelTag
   ) -> (Value:Fort Variant Channel VM Object Reference)
   > Branch(Condition=@n3.ReturnValue)
   > HandleVariantEngaged!(execute, self:Fort Variant Channel VM Object Reference=@n3, InVariant:Fort Variant Item VM Object Reference=@n1.ItemVM) -> (then)
   > Call On Variant Item Engaged!(
    execute:Exec
    self:WBP UIKit Variant Row Redirector Object Reference
    VariantItem:Fort Variant Item VM Object Reference=@n1.ItemVM
    VariantChannel:Fort Variant Channel VM Object Reference=@n3
    AdditionalInfo:String=Split(SourceString=@n2.CustomData, InStr=:, SearchDir=FromStart).RightS
   ) -> (then:Exec) {is=title K2Node_CallDelegate}
  ]}
 GetDesiredFocusTarget {nodes [
   GetDesiredFocusTarget() {is=title K2Node_FunctionEntry}
   > n1 = Cast To BPI_DynVariantRow_FocusTargetWidget!(
    execute
    Object=GetChildAt!(self:Panel Widget Object Reference=Get Container() -> (Container:Panel Widget Object Reference), Index:int=0) -> (ReturnValue:Widget Object Reference)
   ) -> (then, CastFailed, AsBPI Dyn Variant Row Focus Target Widget:BPI Dyn Variant Row Focus Target Widget Interface) {is=title K2Node_DynamicCast}
   > n2 = GetDesiredFocusTarget!(execute, self:BPI Dyn Variant Row Focus Target Widget Interface=@n1) -> (then, FocusTarget:Widget Object Reference) {is=title K2Node_CallFunction}
   > Return Node!(execute, FocusTarget:Widget Object Reference=@n2)
   Return Node(execute=@n1.CastFailed, FocusTarget=Get Container())
  ]}
 InitializeRow {nodes [
   n1 = InitializeRow!() -> (then, ChannelVM:Array of Fort Variant Channel VM Object References) {is=title K2Node_FunctionEntry}
   > n2 = For Each Loop(Array:Array of Fort Variant Channel VM Object References=@n1) -> (Array Element:Fort Variant Channel VM Object Reference)
   > Add!(
    execute
    TargetMap:Map of Gameplay Tag Structures to Fort Variant Channel VM Object References=Get VariantChannelsMap()
    Key:"Gameplay Tag Structure (by ref)"=Break Mcp Variant Channel Info(McpVariantChannelInfo=Get VariantChannelInfo!(self:Fort Variant Channel VM Object Reference=@n2) -> (VariantChannelInfo:Mcp Variant Channel Info Structure)).VariantChannelTag
    Value:"Fort Variant Channel VM Object Reference (by ref)"=@n2
   ) -> (then) {is=title K2Node_CallFunction}
   > AddVariantChannelVM!(execute, self:Fort Redirector Variant Wrapper Object Reference, ChannelVM:Fort Variant Channel VM Object Reference=@n2) -> (then)
  ]}
 OnVariantItemEngaged {nodes [OnVariantItemEngaged!() -> (then, VariantItem:Fort Variant Item VM Object Reference, VariantChannel:Fort Variant Channel VM Object Reference, AdditionalInfo:String) {is=title K2Node_FunctionEntry}]}
]
variables [
 OnVariantItemEngaged {}
 SprayID {type=string, default=""}
 VariantChannelsMap {default {}}
]
widgets {
 info {
  inheritedWidgetCount=1
  namedSlotCount=1
  parentClass=/Script/FortniteUI.FortRedirectorVariantWrapper
  rootWidgetClass=&NamedSlot
  widgetCount=1
 }
 widgets [Container {widgetClassPath=&NamedSlot, bInherited=true, bIsVariable=true, widget=Container}]
}
