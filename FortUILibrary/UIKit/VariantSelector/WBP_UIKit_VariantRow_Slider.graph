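# Variant selector row widget that lets the player set a float variant channel value with a UIKit slider.
# InitializeRow caches the slider channel view model; slider changes update the progress bar and restart a short debounce timer.
# With real-time preview, each change commits immediately; otherwise the timer commits the pending value.
# When the timer ends, it broadcasts On Variant Item Engaged with the equipped item and slider value.
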WidgetBlueprint /FortUILibrary/UIKit/VariantSelector/WBP_UIKit_VariantRow_Slider
graphs [
 CommitFloatValue {nodes [
   CommitFloatValue() {is=title K2Node_FunctionEntry}
   > SetFloatValue!(
    execute
    self:Fort Variant Slider Channel VM Object Reference=Get FortVariantSliderChannelVM() -> (FortVariantSliderChannelVM:Fort Variant Slider Channel VM Object Reference)
    InFloatValue:float=Get PendingCommitValue() -> (PendingCommitValue:double)
    bCommit:bool=true
   ) -> (then)
   > PrintString!(execute, InString=Commit, bPrintToScreen, bPrintToLog, TextColor=(R=1,G=0.8473,B=0,A=1), Duration, Key) -> (then)
  ]}
 EventGraph {nodes [
   n1 = "On Slider Value Changed (WBP_UIKit_Slider)"!() -> (OutputDelegate, then, Value:double) {is=title K2Node_ComponentBoundEvent}
   > SetProgress!(
    execute
    self:WBP UIKit Progress Bar Base Object Reference=Get SliderProgressBar!(self:WBP UIKit Slider Object Reference=Get WBP_UIKit_Slider() -> (WBP_UIKit_Slider:WBP UIKit Slider Object Reference)) -> (SliderProgressBar:WBP UIKit Progress Bar Object Reference)
    Progress:double=@n1.Value
   ) -> (then)
   > Set PendingCommitValue(PendingCommitValue:double=@n1.Value) -> (Output_Get:double)
   > n2 = Clear and Invalidate Timer by Handle(Handle=Get CommitTimer() -> (CommitTimer:Timer Handle Structure))
   n3 = OnCommitHoldTimerFinished() {is=title K2Node_CustomEvent}
   n4 = Set Timer by Event(execute=@n2.then, Delegate=@n3, Time=Get CommitHoldTime() -> (CommitHoldTime:double), bLooping=false)
   > Set CommitTimer(CommitTimer:Timer Handle Structure=@n4) -> (Output_Get:Timer Handle Structure)
   > Branch(Condition=Get bRealTimePreviewUpdate!(self:Fort Variant Slider Channel VM Object Reference=Get FortVariantSliderChannelVM()) -> (bRealTimePreviewUpdate:bool))
   > SetFloatValue(self=Get FortVariantSliderChannelVM(), InFloatValue=Get PendingCommitValue(), bCommit=true)
   n5 = Get PendingCommitValue()
   n6 = Get FortVariantSliderChannelVM()
   Call On Variant Item Engaged!(
    execute:Exec=[@n7.then, @n8.then]
    self:WBP UIKit Variant Row Base Object Reference
    VariantItem:Fort Variant Item VM Object Reference=Get EquippedVariantItem!(self:Fort Variant Channel VM Object Reference=@n6) -> (EquippedVariantItem:Fort Variant Item VM Object Reference)
    VariantChannel:Fort Variant Channel VM Object Reference=@n6
    AdditionalInfo:String=Append(A="Slider Value ", B="To String (Float)"(InDouble=@n5))
   ) -> (then:Exec) {is=title K2Node_CallDelegate}
   n7 = SetFloatValue(execute=@n8.else, self=@n6, InFloatValue=@n5, bCommit=true)
   n8 = Branch(execute=@n3.then, Condition=Get bRealTimePreviewUpdate(self=Get FortVariantSliderChannelVM()))
  ]}
 GetDesiredFocusTarget {nodes [
   GetDesiredFocusTarget() {is=title K2Node_FunctionEntry}
   > Return Node!(
    execute
    FocusTarget:Widget Object Reference=Get Slider_Value!(self:WBP UIKit Slider Object Reference=Get WBP_UIKit_Slider()) -> (Slider_Value:Analog Slider Object Reference)
   )
  ]}
 InitializeRow {nodes [
   n1 = InitializeRow!() -> (
    then
    ChannelVM:Fort Variant Channel VM Object Reference
    VariantSelectorVM:Fort Variant Selector VM Object Reference
    AdditionalArguments:Map of Names to Booleans
    TargetStyleTag:Gameplay Tag Structure
   ) {is=title K2Node_FunctionEntry}
   > n2 = Cast To FortVariantSliderChannelVM!(execute, Object=@n1) -> (then, CastFailed, AsFort Variant Slider Channel VM:Fort Variant Slider Channel VM Object Reference) {is=title K2Node_DynamicCast}
   > Set FortVariantSliderChannelVM(FortVariantSliderChannelVM:Fort Variant Slider Channel VM Object Reference=@n2) -> (Output_Get:Fort Variant Slider Channel VM Object Reference)
  ]}
]
variables [
 CommitHoldTime {type=number, default=0.25}
 CommitTimer {type=TimerHandle}
 LastValue {type=number, default=-1}
 PendingCommitValue {type=number}
]
widgets {
 info {
  inheritedWidgetCount=0
  namedSlotCount=1
  parentClass=&/WBP_UIKit_VariantRow_Base.WBP_UIKit_VariantRow_Base_C
  rootWidgetClass=null
  widgetCount=1
 }
 widgets [WBP_UIKit_Slider {widgetClassPath=/FortUILibrary/UIKit/Slider/WBP_UIKit_Slider, bIsVariable=true, widget=WBP_UIKit_Slider}]
}
