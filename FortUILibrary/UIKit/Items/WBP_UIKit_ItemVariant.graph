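# WBP_UIKit_ItemVariant is an item card list entry widget, derived from WBP_UIKit_ItemCard_Base, that displays a cosmetic variant.
# It stores its FortVariantItemVM, applies texture, tint and material variance, and shows either a locked state or an owned check indicator.
# Hover plays an optional scale animation, and a touch click on an unselected card also focuses and selects it.
# A helper decides age gating from item details and the cosmetic's rating.
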WidgetBlueprint /FortUILibrary/UIKit/Items/WBP_UIKit_ItemVariant
animations [Anim_Hover {
  binding_count=1
  length_seconds=0.2
  object=&self:Anim_Hover
  sequence {bindings [WBP_UIKit_ItemVariant_C_4 {
     id=#1
     possessed_class=WBP_UIKit_ItemVariant_C
     tracks [{
       type=MovieScene2DTransformTrack
       sections [{
         type=MovieScene2DTransformSection
         end_frame=4
         start_frame=0
         channels [
          Translation.X {type=MovieSceneScriptingFloatChannel, const [0,1]}
          Translation.Y {type=MovieSceneScriptingFloatChannel, const [0,1]}
          Angle {type=MovieSceneScriptingFloatChannel, const [0,1]}
          Scale.X {type=MovieSceneScriptingFloatChannel, keys [[0,1],[4,1.1]]}
          Scale.Y {type=MovieSceneScriptingFloatChannel, keys [[0,1],[4,1.1]]}
          Shear.X {type=MovieSceneScriptingFloatChannel, const [0,1]}
          Shear.Y {type=MovieSceneScriptingFloatChannel, const [0,1]}
         ]
        }]
      }]
    }]}
 }]
graphs [
 AllowHoverAnimations {nodes [
   n1 = AllowHoverAnimations!() -> (then, AllowHoverAnimation:bool=~) {is=title K2Node_FunctionEntry}
   > Set AllowHoverAnimation(AllowHoverAnimation:bool=@n1) -> (Output_Get:bool)
  ]}
 EventGraph {nodes [
   n1 = Event PreConstruct()
   > Parent: PreConstruct(IsDesignTime=@n1.IsDesignTime)
   > SetItemIsFavorite!(execute, self:WBP UIKit Item Card Base Object Reference, Is Favorite:bool=false) -> (then)
   > SetItemIsArchived!(execute, self:WBP UIKit Item Card Base Object Reference, Is Archived:bool=false) -> (then)
   > SetItemIsEnhanced!(execute, self:WBP UIKit Item Card Base Object Reference, Is Enhanced:bool=false) -> (then)
   > SetItemIsMultiGame!(execute, self:WBP UIKit Item Card Base Object Reference, Has Transference:bool=false) -> (then)
   > SetItemIsLocked!(execute, self:WBP UIKit Item Card Base Object Reference, Is Locked:bool=false) -> (then)
   > SetItemImage!(
    execute
    self:WBP UIKit Item Card Base Object Reference
    Item Image:Soft Object Reference=Get Fallback Image() -> (Fallback Image:Soft Object Reference)
   ) -> (then)
   Event On Hovered()
   > Parent: On Hovered() {is=title K2Node_CallParentFunction}
   > Branch(Condition=Get AllowHoverAnimation() -> (AllowHoverAnimation:bool))
   > PlayAnimationForward(InAnimation=Get Anim_Hover() -> (Anim_Hover:Widget Animation Object Reference))
   Event On Unhovered()
   > Parent: On Unhovered() {is=title K2Node_CallParentFunction}
   > Branch(Condition=Get AllowHoverAnimation())
   > PlayAnimationReverse(InAnimation=Get Anim_Hover())
   n2 = Event OnListItemObjectSet()
   > n3 = Cast To FortVariantItemVM!(execute, Object=@n2.ListItemObject) -> (then, CastFailed, AsFort Variant Item VM:Fort Variant Item VM Object Reference) {is=title K2Node_DynamicCast}
   > InitializeItemCard(self:WBP UIKit Item Card Base Object Reference)
   > Set FortVariantItemVM(FortVariantItemVM:Fort Variant Item VM Object Reference=@n3) -> (Output_Get:Fort Variant Item VM Object Reference)
   Event On Clicked()
   n4 = GetOwningPlayer()
   n5 = GetSelected!(self:Common Button Base Object Reference) -> (ReturnValue:bool)
   > Branch(Condition=AND Boolean(A=NOT Boolean(A=@n5), B=IsUsingTouch(PlayerController=@n4)))
   > PlayAnimationForward(InAnimation=Get Anim_Hover())
   > SetFocusState!(execute, self:WBP UIKit Item Card Base Object Reference, IsFocused:bool=true) -> (then)
   > n6 = Cast To ListView!(execute, Object=GetOwningListView!(UserListEntry:User List Entry Interface) -> (ReturnValue:List View Base Object Reference)) -> (then, CastFailed, AsList View:List View Object Reference) {is=title K2Node_DynamicCast}
   > Set Selected Item!(
    execute
    self:List View Object Reference=@n6
    Item:Object Reference=GetListItemObject!(UserObjectListEntry:User Object List Entry Interface) -> (ReturnValue:Object Reference)
   ) -> (then) {is=title K2Node_CallFunction}
  ]}
 GetIsVariantAgeGated {nodes [
   n1 = GetIsVariantAgeGated!() -> (then, IsAgeGated:bool=~, CosmeticItem:Fort Item VM Object Reference) {is=title K2Node_FunctionEntry}
   > n2 = Cast To AthenaCosmeticItemDefinition!(
    execute
    Object=Get ItemDefinition!(self:Fort Item VM Object Reference=@n1.CosmeticItem) -> (ItemDefinition:Fort Item Definition Object Reference)
   ) -> (then, CastFailed, AsAthena Cosmetic Item Definition:Athena Cosmetic Item Definition Object Reference) {is=title K2Node_DynamicCast}
   > n3 = ConvertCosmeticItemDefinitionToInterface!(execute, CosmeticItemDefinition:Athena Cosmetic Item Definition Object Reference=@n2) -> (then, ReturnValue:Cosmetic Item Gating Interface Interface)
   > n4 = ShouldBeGatedByAgeRating!(
    execute
    InLocalPlayer:Local Player Object Reference=GetOwningLocalPlayer!(self:Widget Object Reference) -> (ReturnValue:Local Player Object Reference)
    InCosmeticItemGatingInterface:"Cosmetic Item Gating Interface Interface (by ref)"=@n3
    OptionalQueryContextTags:Gameplay Tag Container Structure=(GameplayTags=((TagName="Cosmetics.Gating.QueryContext.ItemVariant")))
   ) -> (then, ReturnValue:bool)
   n5 = Get ItemDetails!(self:Fort Item VM Object Reference=@n1.CosmeticItem) -> (ItemDetails:Fort Item Details VM Object Reference)
   > Return Node!(
    execute
    IsAgeGatedOut:bool=OR Boolean!(A=@n1, B=Get bIsAgeGated!(self:Fort Item Details VM Object Reference=@n5) -> (bIsAgeGated:bool), C:bool=@n4) -> (ReturnValue)
   )
  ]}
 OwnedVisibility {nodes [
   n1 = OwnedVisibility!() -> (then, IsOwned:bool=~) {is=title K2Node_FunctionEntry}
   n2 = Get bIsOwnershipVisibilityAllowed!(self:Fort Variant Item VM Object Reference=Get FortVariantItemVM() -> (FortVariantItemVM:Fort Variant Item VM Object Reference)) -> (bIsOwnershipVisibilityAllowed:bool)
   > Return Node!(execute, NewParam:ESlateVisibility Enum=Select(Option 0=Collapsed, Option 1=SelfHitTestInvisible, Index=AND Boolean(A=@n1, B=@n2)))
  ]}
 SetItemIsDisabled {nodes [
   n1 = SetItemIsDisabled!() -> (then, Disabled:bool=~) {is=title K2Node_FunctionEntry}
   > SetRenderOpacity(InOpacity=SelectFloat(A=0.25, B=1, bPickA=@n1))
  ]}
 SetItemIsOwned {nodes [
   SetItemIsOwned!() -> (then, IsOwned:bool=~) {is=title K2Node_FunctionEntry}
   > n1 = Branch(Condition=Get bShowUnownedAsLocked!(self:Fort Variant Item VM Object Reference=Get FortVariantItemVM()) -> (bShowUnownedAsLocked:bool))
   SetItemIsLocked(execute=[@n1.then, @n2.then], Is Locked=NOT Boolean(A=Get IsOwned() -> (IsOwned:bool)))
   n2 = Branch(execute=@n1.else, Condition=Get bIsLocked!(self:Fort Variant Item VM Object Reference=Get FortVariantItemVM()) -> (bIsLocked:bool))
   n3 = Get IsOwned()
   n4 = Get bIsOwnershipVisibilityAllowed(self=Get FortVariantItemVM())
   SetVisibility(
    execute=@n2.else
    self=Get StatusIndicator_Owned() -> (StatusIndicator_Owned:WBP UIKit Status Indicator Object Reference)
    InVisibility=Select(Option 0=Collapsed, Option 1=SelfHitTestInvisible, Index=AND Boolean(A=@n3, B=@n4))
   )
  ]}
 SetVariantDisplayInfo {nodes [
   n1 = SetVariantDisplayInfo!() -> (then, VariantDisplayInfo:Variant Display Info Structure) {is=title K2Node_FunctionEntry}
   > n2 = Sequence()
   n3 = Break Variant Display Info!(VariantDisplayInfo:"Variant Display Info Structure (by ref)"=@n1) -> (Texture:Texture 2D Soft Object Reference, TintColor:Slate Color Structure, bApplyTexture:bool=~, bApplyTint:bool=~) {is=title K2Node_BreakStruct}
   > n4 = Branch(Condition=@n3.bApplyTexture)
   > n5 = SetItemImage(Item Image=@n3)
   Branch(execute=[@n5.then, @n4.else], Condition=@n3.bApplyTint)
   > Set Item Color!(execute, self:WBP UIKit Item Card Base Object Reference, VariantColor:Slate Color Structure=@n3.TintColor) -> (then) {is=title K2Node_CallFunction}
   ApplyMaterialVariance!(
    execute=@n2.then_1
    self:Fort Variant Item VM Object Reference=Get FortVariantItemVM()
    Image:Image Object Reference=Get LazyImage_Item() -> (LazyImage_Item:Fort Lazy Image Object Reference)
   ) -> (then)
  ]}
]
variables [AllowHoverAnimation {type=boolean, default=true}]
widgets {
 info {
  inheritedWidgetCount=0
  namedSlotCount=1
  parentClass=/FortUILibrary/Systems/Items/WBP_UIKit_ItemCard_Base.WBP_UIKit_ItemCard_Base_C
  rootWidgetClass=null
  widgetCount=2
 }
 widgets [
  GridPanel_Owned {widgetClassPath=&GridPanel, widget=GridPanel_Owned, properties {columnFill [1], rowFill [1]}}
  StatusIndicator_Owned {
   widgetClassPath=/FortUILibrary/UIKit/StatusIndicator/WBP_UIKit_StatusIndicator
   bIsVariable=true
   parent=GridPanel_Owned
   widget=StatusIndicator_Owned
   properties {icon=/FortUILibrary/Textures/MaterialInstances/MI_UIKit_Icon_Check, overrideStatusIndicatorVisibility=true, visibility=Hidden}
   slot {class=GridSlot, properties {horizontalAlignment=HAlign_Right, verticalAlignment=VAlign_Top, padding=<ltrb 0 4 4 0>}}
  }
 ]
}
