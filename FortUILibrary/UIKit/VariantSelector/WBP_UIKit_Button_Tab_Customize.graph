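# Tab button widget for the UIKit variant selector, apparently for customization screens, derived from WBP_UIKit_Button_Selectable with no widgets of its own.
# On PreConstruct it casts the background block to the side-nav tab background and sets pixel-based corner radii.
# It then writes a larger pixel corner radius directly to the outline's dynamic material.
# Variables hold bang and status indicator settings, unused in this graph; Construct only calls the parent.
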WidgetBlueprint /FortUILibrary/UIKit/VariantSelector/WBP_UIKit_Button_Tab_Customize
graphs [EventGraph {nodes [
   n1 = Event PreConstruct()
   > Parent: PreConstruct(IsDesignTime=@n1.IsDesignTime)
   > n2 = Cast To WBP_UIKit_Block_BackgroundTabSideNav!(execute, Object=Get Block_Background() -> (Block_Background:WBP UIKit Block Background Base Object Reference)) -> (then, CastFailed, AsWBP UIKit Block Background Tab Side Nav:WBP UIKit Block Background Tab Side Nav Object Reference) {is=title K2Node_DynamicCast}
   > SetCornerRadiuses!(execute, self:WBP UIKit Block Background Tab Side Nav Object Reference=@n2, CornerRadius:float=10, RadiusInPixels:bool=true) -> (then)
   n3 = Get OutlineDynamicMaterial!(self:WBP UIKit Block Background Base Object Reference=Get Block_Background()) -> (OutlineDynamicMaterial:Material Instance Dynamic Object Reference)
   > SetScalarParameterValue(self=@n3, ParameterName=Radius in Pixels, Value=1)
   > SetVectorParameterValue!(execute, self=@n3, ParameterName=Corner Radius, Value_R:float=18, Value_G:float=18, Value_B:float=18, Value_A:float=18) -> (then)
   Event Construct()
   > Parent: Construct()
  ]}]
variables [
 BangHorizontalAlignment {type=EHorizontalAlignment, default=HAlign_Left}
 BangNudge {type=Vector2D, default=<0 0>}
 BangPadding {type=Margin, default=<ltrb 4 4 4 4>}
 BangType {type=EFortBangType, default=Invalid}
 BangVerticalAlignment {type=EVerticalAlignment, default=VAlign_Center}
 Configuration_0 {type=E_UI_StatusIndicator_Configuration, default=Dot}
 EnableCustomLayout {type=boolean, default=true}
 StatusIndicatorClass_0 {type=Class, default=/FortUILibrary/UIKit/StatusIndicator/WBP_UIKit_StatusIndicator.WBP_UIKit_StatusIndicator_C}
 TutorialName {type=string, default=None}
 Type_0 {type=E_UI_StatusIndicator_Type, default=Attention}
]
widgets {
 info {
  inheritedWidgetCount=0
  parentClass=/FortUILibrary/UIKit/Buttons/WBP_UIKit_Button_Selectable.WBP_UIKit_Button_Selectable_C
  rootWidgetClass=null
  widgetCount=0
 }
 widgets []
}
