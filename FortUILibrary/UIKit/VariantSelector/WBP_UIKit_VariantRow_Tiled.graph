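# Variant selector row widget that shows one cosmetic variant channel as a tile view of item cards.
# InitializeRow sets tile size, spacing, roundness and text case from the channel's custom style, else from channel type and style flags.
# Hover previews variants and unhover restores the equipped one, broadcasting conflicts.
# Clicks engage unless conflicted or locked; on touch, tapping an unselected item only previews it.
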WidgetBlueprint /FortUILibrary/UIKit/VariantSelector/WBP_UIKit_VariantRow_Tiled
graphs [
 EngageItem {nodes [
   n1 = EngageItem!() -> (then, VariantItem:Object Reference) {is=title K2Node_FunctionEntry}
   > n2 = Cast To FortVariantItemVM!(execute, Object=@n1) -> (then, CastFailed, AsFort Variant Item VM:Fort Variant Item VM Object Reference) {is=title K2Node_DynamicCast}
   > Call On Conflicts Updated!(
    execute:Exec
    self:WBP UIKit Variant Row Base Object Reference
    Conflicts:Map of E_VariantConflictType Enums to Texts=GetConflictsFromVariant!(
     SelectedVariant:Fort Variant Item VM Object Reference=@n2
     VariantSelectorVM:Fort Variant Selector VM Object Reference=Get FortVariantSelectorVM() -> (FortVariantSelectorVM:Fort Variant Selector VM Object Reference)
    ) -> (Conflicts:Map of E_VariantConflictType Enums to Texts=NewEnumerator6)
   ) -> (then:Exec) {is=title K2Node_CallDelegate}
   > n3 = Branch(Condition=GetHasConflict!(self, VariantItemVM:Fort Variant Item VM Object Reference=@n2) -> (ReturnValue:bool))
   n4 = Get FortVariantChannelVM() -> (FortVariantChannelVM:Fort Variant Channel VM Object Reference)
   HandleVariantEngaged!(execute=@n3.else, self:Fort Variant Channel VM Object Reference=@n4, InVariant:Fort Variant Item VM Object Reference=@n2) -> (then)
   n5 = "To String (Text)"(InText=Get RowDisplayName!(self:Fort Variant Channel VM Object Reference=@n4) -> (RowDisplayName:Text))
   n6 = Get!(self:Fort Variant Item VM Object Reference=@n2) -> (
    VariantDisplayInfo_Texture:Texture 2D Soft Object Reference
    VariantDisplayInfo_TintColor_SpecifiedColor:Linear Color Structure
    VariantDisplayInfo_TintColor_ColorUseRule:ESlateColorStylingMode Enum
    VariantDisplayInfo_bApplyTexture:bool=~
    VariantDisplayInfo_bApplyTint:bool=~
   )
   n7 = IsValidSoftObjectReference(SoftObjectReference=@n6)
   n8 = GetSoftPtrAssetName!(Asset:Soft Object Reference=@n6) -> (ReturnValue:String)
   n9 = Append(A="HEX ##", B=ToHex!(InColor:Linear Color Structure=@n6.VariantDisplayInfo_TintColor_SpecifiedColor) -> (ReturnValue:String))
   > Call On Variant Item Engaged!(
    execute:Exec
    self:WBP UIKit Variant Row Base Object Reference
    VariantItem:Fort Variant Item VM Object Reference=@n2
    VariantChannel:Fort Variant Channel VM Object Reference=@n4
    AdditionalInfo:String=Select(
     Option 0:String=SelectString(A=@n8, B=@n9, bPickA=@n7)
     Option 1:String
     Index=Split(SourceString=@n5, InStr="(", SearchDir=FromStart).ReturnValue
    ) -> (ReturnValue:String)
   ) -> (then:Exec) {is=title K2Node_CallDelegate}
  ]}
 EngageSelectedItem {nodes [
   EngageSelectedItem() {is=title K2Node_FunctionEntry}
   n1 = GetSelectedVariantItem!(self:Fort Variant Channel VM Object Reference=Get FortVariantChannelVM()) -> (ReturnValue:Fort Variant Item VM Object Reference)
   > Is Valid(InputObject=@n1)
   > EngageItem!(execute, self, VariantItem:Object Reference=@n1) -> (then) {is=title K2Node_CallFunction}
  ]}
 EventGraph {nodes [
   n1 = "On Entry Initialized (TileView_Variants)"!() -> (OutputDelegate, then, Item:Object Reference, Widget:User Widget Object Reference) {is=title K2Node_ComponentBoundEvent}
   > n2 = Cast To WBP_UIKit_ItemVariant!(execute, Object=@n1.Widget) -> (then, CastFailed, AsWBP UIKit Item Variant:WBP UIKit Item Variant Object Reference) {is=title K2Node_DynamicCast}
   > SetBackgroundBlendMode!(execute, self:WBP UIKit Item Card Base Object Reference=@n2, UseAdaptiveColor:bool=NOT Boolean(A=Get IsRound() -> (IsRound:bool))) -> (then)
   > SetRoundedCornersStyle!(execute, self, VariantCard:WBP UIKit Item Variant Object Reference=@n2) -> (then)
   n3 = "On Item Clicked (TileView_Variants)"!() -> (OutputDelegate, then, Item:Object Reference) {is=title K2Node_ComponentBoundEvent}
   n4 = Get CommonInputSubsystem()
   > n5 = Branch(Condition="Equal (Enum)"(A=GetCurrentInputType(self=@n4), B=Touch))
   n6 = Get FortVariantChannelVM()
   > n7 = Branch(Condition="Equal (Object)"(
     A=Get SelectedVariantItem!(self:Fort Variant Channel VM Object Reference=@n6) -> (SelectedVariantItem:Fort Variant Item VM Object Reference)
     B=@n3.Item
    ))
   EngageItem(execute=[@n7.then, @n5.else], VariantItem=@n3.Item)
   PreviewItem!(execute=@n7.else, self, VariantItem:Object Reference=@n3.Item, ShowPreview:bool=true) -> (then)
   n8 = "On Item Is Hovered Changed (TileView_Variants)"!() -> (OutputDelegate, then, Item:Object Reference, bIsHovered:bool) {is=title K2Node_ComponentBoundEvent}
   > PreviewItem(VariantItem=@n8.Item, ShowPreview=@n8.bIsHovered)
  ]}
 GetDesiredFocusTarget {nodes [
   GetDesiredFocusTarget() {is=title K2Node_FunctionEntry}
   > Return Node!(execute, FocusTarget:Widget Object Reference=Get TileView_Variants() -> (TileView_Variants:Fort Tile View Object Reference))
  ]}
 GetHasConflict {nodes [
   n1 = GetHasConflict!() -> (then, VariantItemVM:Fort Variant Item VM Object Reference) {is=title K2Node_FunctionEntry}
   > Return Node!(
    execute
    ReturnValue:bool=OR Boolean(
     A=Get bHasConflict!(self:Fort Variant Item VM Object Reference=@n1) -> (bHasConflict:bool)
     B=Get bIsLocked!(self:Fort Variant Item VM Object Reference=@n1) -> (bIsLocked:bool)
    )
   )
  ]}
 InitializeRow {nodes [
   n1 = InitializeRow!() -> (
    then
    ChannelVM:Fort Variant Channel VM Object Reference
    VariantSelectorVM:Fort Variant Selector VM Object Reference
    AdditionalArguments:Map of Names to Booleans
    TargetStyleTag:Gameplay Tag Structure
   ) {is=title K2Node_FunctionEntry}
   > Set Additional Arguments(Additional Arguments:Map of Names to Booleans=@n1.AdditionalArguments) -> (Output_Get:Map of Names to Booleans)
   > Set ChannelStyle(ChannelStyle:Variant Channel Style Options Structure) -> (Output_Get:Variant Channel Style Options Structure)
   > Set FortVariantSelectorVM(FortVariantSelectorVM:Fort Variant Selector VM Object Reference=Get VariantSelectorVM() -> (VariantSelectorVM:Fort Variant Selector VM Object Reference)) -> (Output_Get:Fort Variant Selector VM Object Reference)
   > n2 = Set FortVariantChannelVM(FortVariantChannelVM:Fort Variant Channel VM Object Reference=Get ChannelVM() -> (ChannelVM:Fort Variant Channel VM Object Reference)) -> (Output_Get:Fort Variant Channel VM Object Reference)
   n3 = TryGetCustomStyle!(
    self:Fort Variant Channel VM Object Reference=@n2
    RequestTag:Gameplay Tag Structure=Get TargetStyleTag() -> (TargetStyleTag:Gameplay Tag Structure)
   ) -> (OutStyle:Variant Channel Style Options Structure, ReturnValue:bool)
   > n4 = Branch(Condition=@n3.ReturnValue)
   > n5 = Set ChannelStyle!(execute, ChannelStyle=@n3) -> (
    then
    Output_Get_bIsRound:bool=~
    Output_Get_EntrySpacing:float=~
    Output_Get_EntrySize:float=~
    Output_Get_TextPolicy:ETextTransformPolicy Enum
   )
   > Set IsRound(IsRound:bool=@n5) -> (Output_Get:bool)
   n6 = Get TileView_Variants()
   > Set HorizontalEntrySpacing!(execute, HorizontalEntrySpacing:float=@n5.Output_Get_EntrySpacing, self:List View Object Reference=@n6) -> (then, Output_Get:float)
   > Set VerticalEntrySpacing!(execute, VerticalEntrySpacing:float=@n5.Output_Get_EntrySpacing, self:List View Object Reference=@n6) -> (then, Output_Get:float)
   > SetEntryWidth!(execute, self:Tile View Object Reference=@n6, NewWidth:float=@n5.Output_Get_EntrySize) -> (then)
   > SetEntryHeight!(execute, self:Tile View Object Reference=@n6, NewHeight:float=@n5.Output_Get_EntrySize) -> (then)
   > SetTextTransformPolicy!(
    execute
    self:Text Object Reference=Get Text_RowCategoryAndItem() -> (Text_RowCategoryAndItem:Common Text Object Reference)
    InTransformPolicy:ETextTransformPolicy Enum=@n5.Output_Get_TextPolicy
   ) -> (then)
   n7 = Cast To FortVariantRichColorChannelVM!(execute=@n4.else, Object=@n2) -> (then, CastFailed, AsFort Variant Rich Color Channel VM:Fort Variant Rich Color Channel VM Object Reference) {is=title K2Node_DynamicCast}
   n8 = Set IsRound(execute=[@n7.then, @n19.then], IsRound=true)
   n9 = Get Additional Arguments() -> (Additional Arguments:Map of Names to Booleans)
   n10 = Find(TargetMap:Map of Names to Booleans=@n9, Key:"Name (by ref)"=UseCustomizeCharacterStyle) -> (Value:bool)
   n11 = NOT Boolean(A=Get IsRound())
   n12 = Find(TargetMap=@n9, Key=UseCustomizeStyle)
   n13 = Select(
    Option 0:double=Get DefaultEntrySpacing() -> (DefaultEntrySpacing:double)
    Option 1:double=Get CustomizeEntrySpacing() -> (CustomizeEntrySpacing:double)
    Index=OR Boolean(A=AND Boolean(A=@n11, B=@n10), B=@n12)
   ) -> (ReturnValue:double)
   n14 = Get TileView_Variants()
   Set HorizontalEntrySpacing(execute=[@n8.then, @n20.then], HorizontalEntrySpacing=@n13, self=@n14)
   > Set VerticalEntrySpacing(VerticalEntrySpacing=@n13, self=@n14)
   n15 = Get IsRound()
   n16 = Find(TargetMap=Get Additional Arguments(), Key=UseCustomizeCharacterStyle)
   n17 = Select(
    Option 0=Select(Option 0=Get DefaultEntrySize() -> (DefaultEntrySize:double), Option 1=Get CustomizeEntrySize() -> (CustomizeEntrySize:double), Index=@n12)
    Option 1=Get CustomizeCharacterEntrySize() -> (CustomizeCharacterEntrySize:double)
    Index=AND Boolean(A=NOT Boolean(A=@n15), B=@n16)
   )
   > SetEntryWidth(self=@n14, NewWidth=@n17)
   > SetEntryHeight(self=@n14, NewHeight=@n17)
   n18 = Find(TargetMap=Get Additional Arguments(), Key=UseCustomizeStyle)
   > SetTextTransformPolicy(
    self=Get Text_RowCategoryAndItem()
    InTransformPolicy=Select(Option 0:ETextTransformPolicy Enum=ToUpper, Option 1:ETextTransformPolicy Enum=None, Index=OR Boolean(A=@n18, B=@n16)) -> (ReturnValue:ETextTransformPolicy Enum)
   )
   n19 = Cast To FortVariantMaterialParameterSetsChannelVM!(execute=@n7.CastFailed, Object=@n2) -> (then, CastFailed, AsFort Variant Material Parameter Sets Channel VM:Fort Variant Material Parameter Sets Channel VM Object Reference) {is=title K2Node_DynamicCast}
   n20 = Set IsRound(execute=@n19.CastFailed, IsRound=false)
  ]}
 PreviewItem {nodes [
   n1 = PreviewItem!() -> (then, VariantItem:Object Reference, ShowPreview:bool=~) {is=title K2Node_FunctionEntry}
   > n2 = Branch(Condition=@n1.ShowPreview)
   > n3 = Cast To FortVariantItemVM(Object=@n1)
   > Call On Conflicts Updated(Conflicts=GetConflictsFromVariant(SelectedVariant=@n3, VariantSelectorVM=Get FortVariantSelectorVM()) -> (Conflicts=NewEnumerator6))
   > ShowVariantEngaged!(execute, self:Fort Variant Channel VM Object Reference=Get FortVariantChannelVM(), InVariant:Fort Variant Item VM Object Reference=@n3) -> (then)
   > Call On Variant Item Previewed!(execute:Exec, self:WBP UIKit Variant Row Base Object Reference, VariantItem:Fort Variant Item VM Object Reference=@n3) -> (then:Exec) {is=title K2Node_CallDelegate}
   n4 = Get!(execute:Exec=@n2.else, self:Fort Variant Channel VM Object Reference=Get FortVariantChannelVM()) -> (then:Exec, else:Exec, EquippedVariantItem:Fort Variant Item VM Object Reference)
   n5 = Get FortVariantSelectorVM()
   > n6 = Call On Conflicts Updated(Conflicts=GetConflictsFromItem!(
     SelectedItem:Fort Item VM Object Reference=Get CosmeticItem!(self:Fort Variant Item VM Object Reference=@n4) -> (CosmeticItem:Fort Item VM Object Reference)
     VariantSelectorVM:Fort Variant Selector VM Object Reference=@n5
    ) -> (Conflicts:Map of E_VariantConflictType Enums to Texts=NewEnumerator6))
   ShowVariantEngaged(execute=[@n6.then, @n7.then], self=Get FortVariantChannelVM())
   > Call On Variant Item Previewed()
   n7 = Call On Conflicts Updated(execute=@n4.else, Conflicts=GetConflictsFromItem(VariantSelectorVM=@n5) -> (Conflicts=NewEnumerator6))
  ]}
 SelectEquippedFocusItem {nodes [
   SelectEquippedFocusItem() {is=title K2Node_FunctionEntry}
   > n1 = Get(self=Get FortVariantChannelVM())
   > Navigate To Item!(
    execute
    self:List View Object Reference=Get TileView_Variants()
    Item:Object Reference=Get EquippedVariantItem!(self:Fort Variant Channel VM Object Reference=Get FortVariantChannelVM()) -> (EquippedVariantItem:Fort Variant Item VM Object Reference)
   ) -> (then)
   > Return Node!(execute, Success:bool=true)
   n2 = Get TileView_Variants()
   n3 = Branch(execute=@n1.else, Condition="integer > integer"(A=GetNumItems!(self:List View Object Reference=@n2) -> (ReturnValue:int=0)))
   n4 = Get TileView_Variants()
   > Navigate To Item(self=@n4, Item=GetItemAt!(self:List View Object Reference=@n4, Index:int=0) -> (ReturnValue:Object Reference))
   > Return Node(Success=true)
   Return Node(execute=@n3.else, Success=false)
  ]}
 SetRoundedCornersStyle {nodes [
   SetRoundedCornersStyle!() -> (then, VariantCard:WBP UIKit Item Variant Object Reference) {is=title K2Node_FunctionEntry}
   > n1 = Branch(Condition=Find(TargetMap=Get Additional Arguments(), Key=UseCustomizeCharacterStyle))
   n2 = Get VariantCard() -> (VariantCard:WBP UIKit Item Variant Object Reference)
   n3 = Get IsRound()
   > SetCornerRadius!(
    execute
    self:WBP UIKit Item Card Base Object Reference=@n2
    UseOverrideValues:bool=@n3
    RadiusInPixels:bool=false
    CornerRadius_TopLeft:double=0.5
    CornerRadius_TopRight:double=0.5
    CornerRadius_BottomLeft:double=0.5
    CornerRadius_BottomRight:double=0.5
    OutlineCornerOffset:double=0
   ) -> (then)
   > AllowHoverAnimations!(execute, self:WBP UIKit Item Variant Object Reference=@n2, AllowHoverAnimation:bool=@n3) -> (then)
   > n4 = GetDynamicMaterial(self=Get LazyImage_Item!(self:WBP UIKit Item Card Base Object Reference=Get VariantCard()) -> (LazyImage_Item:Fort Lazy Image Object Reference))
   > SetScalarParameterValue(self=@n4, ParameterName=Shape - Roundness, Value="To Float (Boolean)"(InBool=@n3))
   n5 = Find(TargetMap=Get Additional Arguments(), Key=UseCustomizeStyle)
   n6 = Get VariantCard()
   SetCornerRadius(
    execute=@n1.else
    self=@n6
    UseOverrideValues=@n5
    RadiusInPixels=false
    CornerRadius_TopLeft=0.5
    CornerRadius_TopRight=0.5
    CornerRadius_BottomLeft=0.5
    CornerRadius_BottomRight=0.5
    OutlineCornerOffset=0
   )
   > AllowHoverAnimations(self=@n6, AllowHoverAnimation=@n5)
  ]}
]
variables [
 Additional Arguments {type="map<boolean>", default {}}
 B {type=number, default=16}
 ChannelStyle {type=VariantChannelStyleOptions, default {bIsRound=false, entrySize=108, entrySpacing=16, textPolicy=None}}
 IsRound {type=boolean}
]
widgets {
 info {
  inheritedWidgetCount=0
  namedSlotCount=1
  parentClass=&/WBP_UIKit_VariantRow_Base.WBP_UIKit_VariantRow_Base_C
  rootWidgetClass=null
  widgetCount=2
 }
 widgets [
  Overlay_44 {widgetClassPath=&Overlay, widget=Overlay_44}
  TileView_Variants {
   widgetClassPath=/Script/FortniteUI.FortTileView
   bIsVariable=true
   parent=Overlay_44
   widget=TileView_Variants
   properties {
    allowOverscroll=false
    bAllowDragging=false
    bConsumeTouchInputWhenScrolling=false
    bEnableRightClickScrolling=false
    bEnableScrollAnimation=false
    bEnableTouchScrolling=false
    clipping=Inherit
    consumeMouseWheel=Never
    entryHeight=108
    entryWidgetClass=/FortUILibrary/UIKit/Items/WBP_UIKit_ItemVariant.WBP_UIKit_ItemVariant_C
    entryWidth=108
    horizontalEntrySpacing=16
    numDesignerPreviewEntries=12
    tileAlignment=LeftAligned
    verticalEntrySpacing=16
   }
   slot {class=OverlaySlot, properties {horizontalAlignment=HAlign_Fill, padding=<ltrb -7 -7 -7 -7>}}
  }
 ]
}
