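# Widget wrapper that hosts one cosmetic variant row in a named slot, choosing the row class from the first variant channel view model.
# When channels update, it creates that widget, then binds and initializes it as either a Redirector or Base row.
# It re-broadcasts row events, sends rotation requests to the variant set view model, and delegates focus to the row.
# A Tiled row serves as design-time preview.
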WidgetBlueprint /FortUILibrary/UIKit/VariantSelector/WBP_UIKit_VariantRow_Wrapper
graphs [
 EngageSelectedItem {nodes [
   EngageSelectedItem() {is=title K2Node_FunctionEntry}
   > n1 = Cast To WBP_UIKit_VariantRow_Tiled!(execute, Object=Get WrappedRow() -> (WrappedRow:User Widget Object Reference)) -> (then, CastFailed, AsWBP UIKit Variant Row Tiled:WBP UIKit Variant Row Tiled Object Reference) {is=title K2Node_DynamicCast}
   > EngageSelectedItem(self:WBP UIKit Variant Row Tiled Object Reference=@n1) {is=title K2Node_CallFunction}
  ]}
 EventGraph {nodes [
   n1 = HandleOnVariantItemEngaged!() -> (
    OutputDelegate
    then
    VariantItem:Fort Variant Item VM Object Reference
    VariantChannel:Fort Variant Channel VM Object Reference
    AdditionalInfo:String
   ) {is=title K2Node_CustomEvent}
   > Call On Variant Item Engaged!(
    execute:Exec
    self:WBP UIKit Variant Row Wrapper Object Reference
    VariantItem:Fort Variant Item VM Object Reference=@n1.VariantItem
    VariantChannel:Fort Variant Channel VM Object Reference=@n1.VariantChannel
    AdditionalInfo:String=@n1.AdditionalInfo
   ) -> (then:Exec) {is=title K2Node_CallDelegate}
   n2 = HandleOnConflictsUpdated!() -> (OutputDelegate, then, Conflicts:Map of E_VariantConflictType Enums to Texts=NewEnumerator6) {is=title K2Node_CustomEvent}
   > Call On Conflicts Updated!(execute:Exec, self:WBP UIKit Variant Row Wrapper Object Reference, Conflicts:Map of E_VariantConflictType Enums to Texts=@n2.Conflicts) -> (then:Exec) {is=title K2Node_CallDelegate}
   n3 = HandleOnVariantItemPreviewed!() -> (OutputDelegate, then, VariantItem:Fort Variant Item VM Object Reference) {is=title K2Node_CustomEvent}
   > Call On Variant Item Previewed!(execute:Exec, self:WBP UIKit Variant Row Wrapper Object Reference, VariantItem:Fort Variant Item VM Object Reference=@n3.VariantItem) -> (then:Exec) {is=title K2Node_CallDelegate}
   n4 = HandleOnRotationRequested!() -> (OutputDelegate, then, RotationOffset:"Rotator (by ref)", bSnap:bool=~) {is=title K2Node_CustomEvent}
   > HandleRotationRequested!(
    execute
    self:Fort Variant Set VM Object Reference=Get FortVariantSetVM() -> (FortVariantSetVM:Fort Variant Set VM Object Reference)
    RotationOffset:"Rotator (by ref)"=@n4.RotationOffset
    bSnap:bool=@n4.bSnap
   ) -> (then)
   n5 = Event PreConstruct()
   > Branch(Condition=@n5.IsDesignTime)
   > n6 = Create Widget!(
    execute:Exec
    Class:User Widget Class Reference=&/WBP_UIKit_VariantRow_Tiled.WBP_UIKit_VariantRow_Tiled_C
    OwningPlayer:Player Controller Object Reference
    FortVariantChannelVM:Fort Variant Channel VM Object Reference
    FortVariantSelectorVM:Fort Variant Selector VM Object Reference
   ) -> (then:Exec, ReturnValue:WBP UIKit Variant Row Tiled Object Reference)
   > AddChild(self=Get Slot_WrappedRow() -> (Slot_WrappedRow:Named Slot Object Reference), Content=@n6)
  ]}
 GetDesiredFocusTarget {nodes [
   GetDesiredFocusTarget() {is=title K2Node_FunctionEntry}
   n1 = Get WrappedRow()
   > n2 = Cast To WBP_UIKit_VariantRow_Base!(execute, Object=@n1) -> (then, CastFailed, AsWBP UIKit Variant Row Base:WBP UIKit Variant Row Base Object Reference) {is=title K2Node_DynamicCast}
   > Return Node!(
    execute
    FocusTarget:Widget Object Reference=GetDesiredFocusTarget!(self:WBP UIKit Variant Row Base Object Reference=@n2) -> (FocusTarget:Widget Object Reference) {is=title K2Node_CallFunction}
   )
   n3 = Cast To WBP_UIKit_VariantRow_Redirector!(execute=@n2.CastFailed, Object=@n1) -> (then, CastFailed, AsWBP UIKit Variant Row Redirector:WBP UIKit Variant Row Redirector Object Reference) {is=title K2Node_DynamicCast}
   > Return Node(FocusTarget=GetDesiredFocusTarget(self:WBP UIKit Variant Row Redirector Object Reference=@n3))
   Return Node(execute=@n3.CastFailed, FocusTarget=Get Slot_WrappedRow())
  ]}
 OnConflictsUpdated {nodes [OnConflictsUpdated!() -> (then, Conflicts:Map of E_VariantConflictType Enums to Texts=NewEnumerator6) {is=title K2Node_FunctionEntry}]}
 OnVariantItemEngaged {nodes [OnVariantItemEngaged!() -> (then, VariantItem:Fort Variant Item VM Object Reference, VariantChannel:Fort Variant Channel VM Object Reference, AdditionalInfo:String) {is=title K2Node_FunctionEntry}]}
 OnVariantItemPreviewed {nodes [OnVariantItemPreviewed!() -> (then, VariantItem:Fort Variant Item VM Object Reference) {is=title K2Node_FunctionEntry}]}
 SelectEquippedFocusTarget {nodes [
   SelectEquippedFocusTarget() {is=title K2Node_FunctionEntry}
   > n1 = Cast To WBP_UIKit_VariantRow_Tiled(Object=Get WrappedRow())
   > SelectEquippedFocusItem!(execute, self:WBP UIKit Variant Row Tiled Object Reference=@n1) -> (then, Success:bool)
   SetFocus(execute=@n1.CastFailed, self=GetDesiredFocusTarget(self:Self Object Reference))
  ]}
 VM_VariantChannelsUpdated {nodes [
   n1 = VM_VariantChannelsUpdated!() -> (then, ChannelVMs:Array of Fort Variant Channel VM Object References) {is=title K2Node_FunctionEntry}
   > n2 = Set FirstChannelVM(FirstChannelVM:Fort Variant Channel VM Object Reference="Get (a copy)"(Array:Array of Fort Variant Channel VM Object References=@n1) -> (Output:Fort Variant Channel VM Object Reference)) -> (Output_Get:Fort Variant Channel VM Object Reference)
   > ClearChildren(self=Get Slot_WrappedRow())
   > n3 = LoadClassAsset_Blocking(AssetClass=GetWidgetClass!(self:Fort Variant Channel VM Object Reference=@n2) -> (ReturnValue:User Widget Soft Class Reference))
   > n4 = Cast To UserWidget Class!(execute:Exec, Class:Object Class Reference=@n3) -> (then:Exec, CastFailed:Exec, AsUser Widget:User Widget Class Reference) {is=title K2Node_ClassDynamicCast}
   > n5 = Create Widget!(execute, Class=@n4, OwningPlayer) -> (then, ReturnValue:User Widget Object Reference)
   > n6 = Set WrappedRow(WrappedRow:User Widget Object Reference=@n5) -> (Output_Get:User Widget Object Reference)
   > AddChild(self=Get Slot_WrappedRow(), Content=@n6)
   > n7 = Cast To WBP_UIKit_VariantRow_Redirector(Object=Get WrappedRow())
   > Unbind all Events from On Rotation Requested Event!(execute:Exec, self:Fort Redirector Variant Wrapper Object Reference=@n7) -> (then:Exec) {is=title K2Node_ClearDelegate}
   > Bind Event to On Rotation Requested Event(self:Fort Redirector Variant Wrapper Object Reference=@n7, Delegate=Create Event()) {is=title K2Node_AddDelegate}
   > Unbind all Events from On Variant Item Engaged!(execute:Exec, self:WBP UIKit Variant Row Redirector Object Reference=@n7) -> (then:Exec) {is=title K2Node_ClearDelegate}
   > Bind Event to On Variant Item Engaged(self:WBP UIKit Variant Row Redirector Object Reference=@n7, Delegate=Create Event()) {is=title K2Node_AddDelegate}
   > InitializeRow!(
    execute
    self:WBP UIKit Variant Row Redirector Object Reference=@n7
    ChannelVM:Array of Fort Variant Channel VM Object References=Get ChannelVMs() -> (ChannelVMs:Array of Fort Variant Channel VM Object References)
   ) -> (then)
   n8 = Cast To WBP_UIKit_VariantRow_Base(execute=@n7.CastFailed, Object=Get WrappedRow())
   > Unbind all Events from On Conflicts Updated!(execute:Exec, self:WBP UIKit Variant Row Base Object Reference=@n8) -> (then:Exec) {is=title K2Node_ClearDelegate}
   > Bind Event to On Conflicts Updated(self:WBP UIKit Variant Row Base Object Reference=@n8, Delegate=Create Event()) {is=title K2Node_AddDelegate}
   > Unbind all Events from On Variant Item Engaged(self:WBP UIKit Variant Row Base Object Reference=@n8)
   > Bind Event to On Variant Item Engaged(self:WBP UIKit Variant Row Base Object Reference=@n8, Delegate=Create Event())
   > Unbind all Events from On Variant Item Previewed!(execute:Exec, self:WBP UIKit Variant Row Base Object Reference=@n8) -> (then:Exec) {is=title K2Node_ClearDelegate}
   > Bind Event to On Variant Item Previewed(self:WBP UIKit Variant Row Base Object Reference=@n8, Delegate=Create Event()) {is=title K2Node_AddDelegate}
   n9 = Get Additional Arguments() -> (Additional Arguments:Map of Names to Booleans)
   > InitializeRow!(
    execute
    self:WBP UIKit Variant Row Base Object Reference=@n8
    ChannelVM:Fort Variant Channel VM Object Reference=Get FirstChannelVM() -> (FirstChannelVM:Fort Variant Channel VM Object Reference)
    VariantSelectorVM:Fort Variant Selector VM Object Reference=Get FortVariantSelectorVM() -> (FortVariantSelectorVM:Fort Variant Selector VM Object Reference)
    AdditionalArguments:Map of Names to Booleans=@n9
    TargetStyleTag:Gameplay Tag Structure=Get TargetStyleTag() -> (TargetStyleTag:Gameplay Tag Structure)
   ) -> (then)
   > UseCustomizeMenuStyle!(
    execute
    self:WBP UIKit Variant Row Base Object Reference=@n8
    UseCustomizeMenuStyle:bool=OR Boolean(
     A=Find(TargetMap:Map of Names to Booleans=@n9, Key:"Name (by ref)"=UseCustomizeStyle) -> (Value:bool)
     B=Find(TargetMap=@n9, Key=UseCustomizeCharacterStyle)
    )
   ) -> (then)
  ]}
]
variables [
 Additional Arguments {type="map<boolean>", default {}}
 OnConflictsUpdated {}
 OnVariantItemEngaged {}
 OnVariantItemPreviewed {}
 TargetStyleTag {type=GameplayTag, default {tagName=None}}
 WrappedRow {type=UserWidget, default=null}
]
widgets {
 info {inheritedWidgetCount=0, parentClass=&UserWidget, rootWidgetClass=&NamedSlot, widgetCount=1}
 widgets [Slot_WrappedRow {widgetClassPath=&NamedSlot, bIsVariable=true, widget=Slot_WrappedRow}]
}
