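# UIKit variant selector widget for cosmetic customization, with primary and secondary filter tabs above or beside channel rows of variants.
# FortVariantSelectorVM updates rebuild the tabs and a DynamicEntryBox of row wrappers, which relay preview, engage and conflict events outward.
# Filter changes reset previews to equipped variants and show a throbber until variant sets arrive.
# It also sets explicit row navigation, focuses equipped variants on gamepad, and supports horizontal and mobile layouts.
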WidgetBlueprint /FortUILibrary/UIKit/VariantSelector/WBP_UIKit_VariantSelector
graphs [
 ApplyVariantOptions {nodes [
   ApplyVariantOptions() {is=title K2Node_FunctionEntry}
   n1 = Get FortVariantSelectorVM() -> (FortVariantSelectorVM:Fort Variant Selector VM Object Reference)
   > Set bAllowDisplayUnownedVariants!(
    execute
    bAllowDisplayUnownedVariants:bool=Get DisplayUnownedVariants() -> (DisplayUnownedVariants:bool)
    self:Fort Variant Selector VM Object Reference=@n1
   ) -> (then, Output_Get:bool)
   > Set bShowUnownedAsLocked!(
    execute
    bShowUnownedAsLocked:bool=Get ShowUnownedVariantsAsLocked() -> (ShowUnownedVariantsAsLocked:bool)
    self:Fort Variant Selector VM Object Reference=@n1
   ) -> (then, Output_Get:bool)
   > Set bShowAgeGateWarnings!(
    execute
    bShowAgeGateWarnings:bool=Get ShowAgeGateWarnings() -> (ShowAgeGateWarnings:bool)
    self:Fort Variant Selector VM Object Reference=@n1
   ) -> (then, Output_Get:bool)
  ]}
 EngageAllSelectedItems {nodes [
   EngageAllSelectedItems() {is=title K2Node_FunctionEntry}
   > n1 = For Each Loop(Array:Array of User Widget Object References=GetAllEntries!(self:Dynamic Entry Box Base Object Reference=Get EntryBox_ChannelRows() -> (EntryBox_ChannelRows:Dynamic Entry Box Object Reference)) -> (ReturnValue:Array of User Widget Object References)) -> (Array Element:User Widget Object Reference)
   > n2 = Cast To WBP_UIKit_VariantRow_Wrapper!(execute, Object=@n1) -> (then, CastFailed, AsWBP UIKit Variant Row Wrapper:WBP UIKit Variant Row Wrapper Object Reference) {is=title K2Node_DynamicCast}
   > EngageSelectedItem(self:WBP UIKit Variant Row Wrapper Object Reference=@n2)
  ]}
 EventGraph {nodes [
   n1 = HandleOnVariantItemEngaged!() -> (
    OutputDelegate
    then
    VariantItem:Fort Variant Item VM Object Reference
    VariantChannel:Fort Variant Channel VM Object Reference
    AdditionalInfo:String
   ) {is=title K2Node_CustomEvent}
   > UpdateVariantChannelConflicts(self:Fort Variant Selector VM Object Reference=Get FortVariantSelectorVM())
   > Call On Variant Item Engaged!(
    execute:Exec
    self:WBP UIKit Variant Selector Object Reference
    VariantItem:Fort Variant Item VM Object Reference=@n1.VariantItem
    VariantChannel:Fort Variant Channel VM Object Reference=@n1.VariantChannel
    AdditionalInfo:String=@n1.AdditionalInfo
   ) -> (then:Exec) {is=title K2Node_CallDelegate}
   n2 = OnSecondaryTabSelected!() -> (OutputDelegate, then, TabID:Name) {is=title K2Node_CustomEvent}
   > ResetPreviewToEquippedVariants()
   > SetActiveWidget(
    self=Get Switcher_Content() -> (Switcher_Content:Common Visibility Switcher Object Reference)
    Widget=Get Throbber_Content() -> (Throbber_Content:WBP UIKit Throbber Object Reference)
   )
   n3 = Get Tabs_SecondaryFilter() -> (Tabs_SecondaryFilter:WBP UIKit Tabs Object Reference)
   n4 = Get FortVariantSelectorVM()
   > SetCurrentSecondaryFilter!(
    execute
    self:Fort Variant Selector VM Object Reference=Get FortVariantSelectorVM()
    InCurrentSecondaryFilter:Fort Variant Selector Sub Filter VM Object Reference="Get (a copy)"(
     Array:Array of Fort Variant Selector Sub Filter VM Object References=Get SecondaryFilters!(self:Fort Variant Selector VM Object Reference=@n4) -> (SecondaryFilters:Array of Fort Variant Selector Sub Filter VM Object References)
     Dimension 1=GetTabIndexByID!(self:UIKit Tab Group Object Reference=@n3, TabNameID:Name=@n2.TabID) -> (ReturnValue:int=0)
    ) -> (Output:Fort Variant Selector Sub Filter VM Object Reference)
   ) -> (then)
   > Call On Filter Changed!(execute:Exec, self:WBP UIKit Variant Selector Object Reference) -> (then:Exec) {is=title K2Node_CallDelegate}
   n5 = HandleOnVariantItemPreviewed!() -> (OutputDelegate, then, VariantItem:Fort Variant Item VM Object Reference) {is=title K2Node_CustomEvent}
   > Call On Variant Item Previewed!(execute:Exec, self:WBP UIKit Variant Selector Object Reference, VariantItem:Fort Variant Item VM Object Reference=@n5.VariantItem) -> (then:Exec) {is=title K2Node_CallDelegate}
   n6 = HandleOnSelectedItemSet!() -> (OutputDelegate, then, SelectedItem:Fort Item VM Object Reference) {is=title K2Node_CustomEvent}
   > DismissVariantChannels(self:Fort Variant Selector VM Object Reference=Get FortVariantSelectorVM())
   > SetVariantPropertyQueryFilter!(
    execute
    self:Fort Variant Selector VM Object Reference=Get FortVariantSelectorVM()
    InTag:Gameplay Tag Structure=Get VariantPropertyFilterTag() -> (VariantPropertyFilterTag:Gameplay Tag Structure)
    bForceUpdate:bool=false
   ) -> (then)
   > UpdateVariantChannels(self:Fort Variant Selector VM Object Reference=Get FortVariantSelectorVM())
   > HandleOnConflictsUpdated!(
    execute
    self
    Conflicts:Map of E_VariantConflictType Enums to Texts=GetConflictsFromItem!(
     SelectedItem:Fort Item VM Object Reference=@n6.SelectedItem
     VariantSelectorVM:Fort Variant Selector VM Object Reference=Get FortVariantSelectorVM()
    ) -> (Conflicts:Map of E_VariantConflictType Enums to Texts=NewEnumerator6)
   ) -> (then)
   n7 = HandleOnConflictsUpdated!() -> (OutputDelegate, then, Conflicts:Map of E_VariantConflictType Enums to Texts=NewEnumerator6) {is=title K2Node_CustomEvent}
   n8 = GetPrioritizedConflictFromMap!(Conflicts:Map of E_VariantConflictType Enums to Texts=@n7.Conflicts) -> (ConflictText:Text, ConflictType:E_VariantConflictType Enum=NewEnumerator6)
   > Call On Conflict Text Updated!(
    execute:Exec
    self:WBP UIKit Variant Selector Object Reference
    ConflictText:Text=@n8
    ConflictType:E_VariantConflictType Enum=@n8.ConflictType
   ) -> (then:Exec) {is=title K2Node_CallDelegate}
   n9 = OnPrimaryTabSelected!() -> (OutputDelegate, then, TabID:Name) {is=title K2Node_CustomEvent}
   > ResetPreviewToEquippedVariants()
   > SetActiveWidget(self=Get Switcher_Content(), Widget=Get Throbber_Content())
   n10 = Get Tabs_PrimaryFilter() -> (Tabs_PrimaryFilter:WBP UIKit Tabs Object Reference)
   n11 = Get FortVariantSelectorVM()
   > SetCurrentPrimaryFilter!(
    execute
    self:Fort Variant Selector VM Object Reference=Get FortVariantSelectorVM()
    InCurrentPrimaryFilter:Fort Variant Selector Filter VM Object Reference="Get (a copy)"(
     Array:Array of Fort Variant Selector Filter VM Object References=Get PrimaryFilters!(self:Fort Variant Selector VM Object Reference=@n11) -> (PrimaryFilters:Array of Fort Variant Selector Filter VM Object References)
     Dimension 1=GetTabIndexByID(self=@n10, TabNameID=@n9.TabID) -> (ReturnValue=0)
    ) -> (Output:Fort Variant Selector Filter VM Object Reference)
   ) -> (then)
   > SelectTabByIndex!(execute, self:UIKit Tab Group Object Reference=Get Tabs_SecondaryFilter(), Index:int=0, bSuppressClickFeedback:bool=true) -> (then, ReturnValue:bool)
   > Call On Filter Changed()
   n12 = HandleOnRotationRequested!() -> (OutputDelegate, then, RotationOffset:Rotator=~, Snap:bool=~) {is=title K2Node_CustomEvent}
   > Call On Rotation Requested!(execute:Exec, self:WBP UIKit Variant Selector Object Reference, RotationOffset:Rotator=@n12.RotationOffset, Snap:bool=@n12.Snap) -> (then:Exec) {is=title K2Node_CallDelegate}
   Event PreConstruct()
   > SetupLayout!(execute, self, UseCustomizeLayout:bool=Get Use Horizontal Layout() -> (Use Horizontal Layout:bool)) -> (then)
   Event Destruct()
   > Unbind all Events from On Tab Selected!(execute:Exec, self:Common Tab List Widget Base Object Reference=Get Tabs_PrimaryFilter()) -> (then:Exec) {is=title K2Node_ClearDelegate}
   > Unbind all Events from On Tab Selected(self=Get Tabs_SecondaryFilter())
   Event Construct()
   > n13 = Get!(execute:Exec) -> (then:Exec, else:Exec, FortVariantSelectorVM:Fort Variant Selector VM Object Reference)
   > ApplyVariantOptions() {is=title K2Node_CallFunction}
   > SetConfigurationData!(
    execute
    self:Fort Variant Selector VM Object Reference=@n13
    InConfigurationData:Athena Variant Picker Config Object Reference=/Game/Athena/UI/Customization/FortVariantPickerConfig
    bForceUpdate:bool=false
   ) -> (then)
   n14 = Get FortVariantSelectorVM()
   > Unbind all Events from On Rotation Requested Event!(execute:Exec, self:Fort Variant Selector VM Object Reference=@n14) -> (then:Exec) {is=title K2Node_ClearDelegate}
   > Bind Event to On Rotation Requested Event(self:Fort Variant Selector VM Object Reference=@n14, Delegate=Create Event()) {is=title K2Node_AddDelegate}
   > SetMobileTabSizing()
   > SetInputActionsType!(
    execute
    self:WBP UIKit Tabs Base Object Reference=Get Tabs_SecondaryFilter()
    InputActions:E_UI_TabsInputActionType Enum=NewEnumerator0
    InputActionsWithMouseScroll:bool=false
   ) -> (then)
  ]}
 GetDesiredFocusTarget {nodes [
   GetDesiredFocusTarget() {is=title K2Node_FunctionEntry}
   n1 = GetAllEntries(self=Get EntryBox_ChannelRows())
   > n2 = Branch(Condition=Is Valid Index!(TargetArray:Array of User Widget Object References=@n1, IndexToTest:int) -> (ReturnValue:bool))
   > n3 = Cast To WBP_UIKit_VariantRow_Wrapper(Object="Get (a copy)"(Array:Array of User Widget Object References=@n1) -> (Output:User Widget Object Reference))
   > Return Node!(
    execute
    ReturnValue:Widget Object Reference=GetDesiredFocusTarget!(self:WBP UIKit Variant Row Wrapper Object Reference=@n3) -> (FocusTarget:Widget Object Reference) {is=title K2Node_CallFunction}
   )
   Return Node(execute=[@n3.CastFailed, @n2.else])
  ]}
 GetIcon {nodes [
   n1 = GetIcon!() -> (then, Asset:Soft Object Reference) {is=title K2Node_FunctionEntry}
   > n2 = LoadAsset_Blocking(Asset=Select(
     Option 0:Soft Object Reference=Get DefaultFilterIcon() -> (DefaultFilterIcon:Texture 2D Soft Object Reference)
     Option 1:Soft Object Reference=@n1
     Index=IsValidSoftObjectReference(SoftObjectReference=@n1)
    ) -> (ReturnValue:Soft Object Reference))
   > n3 = Cast To Texture2D(Object=@n2)
   > n4 = Set TempTexture(TempTexture:Texture 2D Object Reference=@n3) -> (Output_Get:Texture 2D Object Reference)
   Return Node!(execute=[@n8.then, @n4.then], Texture:Texture 2D Object Reference=Get TempTexture() -> (TempTexture:Texture 2D Object Reference))
   n5 = Cast To MaterialInstance!(execute=@n3.CastFailed, Object=@n2) -> (then, CastFailed, AsMaterial Instance:Material Instance Object Reference) {is=title K2Node_DynamicCast}
   n6 = Get TextureParameterValues!(self:Material Instance Object Reference=@n5) -> (TextureParameterValues:Array of Texture Parameter Value Structures)
   > n7 = Cast To Texture2D(Object=Break Texture Parameter Value!(TextureParameterValue:"Texture Parameter Value Structure (by ref)"="Get (a ref)"(Array:Array of Texture Parameter Value Structures=@n6, Dimension 1=0) -> (Output:"Texture Parameter Value Structure (by ref)")) -> (ParameterInfo:Material Parameter Info Structure, ParameterValue:Texture Object Reference) {is=title K2Node_BreakStruct}.ParameterValue)
   > n8 = Set TempTexture(TempTexture=@n7)
  ]}
 HandleNextTab_Primary {nodes [
   HandleNextTab_Primary() {is=title K2Node_FunctionEntry}
   > SelectNextTab!(execute, self:UIKit Tab Group Object Reference=Get Tabs_PrimaryFilter(), bSuppressClickFeedback:bool=true) -> (then)
   > Return Node!(execute, bPassThrough:bool=false)
  ]}
 HandlePrevTab_Primary {nodes [
   HandlePrevTab_Primary() {is=title K2Node_FunctionEntry}
   > SelectPreviousTab!(execute, self:UIKit Tab Group Object Reference=Get Tabs_PrimaryFilter(), bSuppressClickFeedback:bool=true) -> (then)
   > Return Node(bPassThrough=false)
  ]}
 HandleStwNextTabInput {nodes [
   HandleStwNextTabInput() {is=title K2Node_FunctionEntry}
   > SelectNextTab(self=Get Tabs_SecondaryFilter(), bSuppressClickFeedback=true)
   > Return Node!(execute, Pass Through:bool=false)
  ]}
 HandleStwPreviousTabInput {nodes [
   HandleStwPreviousTabInput() {is=title K2Node_FunctionEntry}
   > SelectPreviousTab(self=Get Tabs_SecondaryFilter(), bSuppressClickFeedback=true)
   > Return Node(Pass Through=false)
  ]}
 OnConflictTextUpdated {nodes [OnConflictTextUpdated!() -> (then, ConflictText:Text, ConflictType:E_VariantConflictType Enum=NewEnumerator6) {is=title K2Node_FunctionEntry}]}
 OnFilterChanged {nodes [OnFilterChanged() {is=title K2Node_FunctionEntry}]}
 OnRotationRequested {nodes [OnRotationRequested!() -> (then, RotationOffset:Rotator=~, Snap:bool=~) {is=title K2Node_FunctionEntry}]}
 OnVariantItemEngaged {nodes [OnVariantItemEngaged!() -> (then, VariantItem:Fort Variant Item VM Object Reference, VariantChannel:Fort Variant Channel VM Object Reference, AdditionalInfo:String) {is=title K2Node_FunctionEntry}]}
 OnVariantItemPreviewed {nodes [OnVariantItemPreviewed!() -> (then, VariantItem:Fort Variant Item VM Object Reference) {is=title K2Node_FunctionEntry}]}
 ResetPreviewToEquippedVariants {nodes [
   ResetPreviewToEquippedVariants() {is=title K2Node_FunctionEntry}
   > n1 = For Each Loop(Array:Array of Fort Variant Channel VM Object References=Get VariantChannels!(self:Fort Variant Selector VM Object Reference=Get FortVariantSelectorVM()) -> (VariantChannels:Array of Fort Variant Channel VM Object References)) -> (Array Element:Fort Variant Channel VM Object Reference)
   > ShowVariantEngaged!(execute, self:Fort Variant Channel VM Object Reference=@n1, InVariant:Fort Variant Item VM Object Reference) -> (then)
  ]}
 SetAdditionalArguments {nodes [
   n1 = SetAdditionalArguments!() -> (then, Additional Arguments:Map of Names to Booleans) {is=title K2Node_FunctionEntry}
   > Set Additional Arguments(Additional Arguments:Map of Names to Booleans=@n1) -> (Output_Get:Map of Names to Booleans)
  ]}
 SetAllRowsNavigation {nodes [
   SetAllRowsNavigation() {is=title K2Node_FunctionEntry}
   > n1 = Set Entries(Entries:Array of User Widget Object References=GetAllEntries(self=Get EntryBox_ChannelRows())) -> (Output_Get:Array of User Widget Object References)
   > n2 = For Each Loop(Array:Array of User Widget Object References=@n1) -> (Array Element:User Widget Object Reference)
   > SetCustomRowNavigation!(execute, self, TargetRow:User Widget Object Reference=@n2, TargetRowIndex:int=@n2.Array Index, Direction:EUINavigation Enum=Up) -> (then)
   > SetCustomRowNavigation(TargetRow=@n2, TargetRowIndex=@n2.Array Index, Direction=Down)
   Cast To WBP_UIKit_VariantRow_Wrapper(Object:Wildcard)
  ]}
 SetCustomRowNavigation {nodes [
   SetCustomRowNavigation!() -> (then, TargetRow:User Widget Object Reference, TargetRowIndex:int, Direction:EUINavigation Enum) {is=title K2Node_FunctionEntry}
   > Set Entries(Entries=GetAllEntries(self=Get EntryBox_ChannelRows()))
   n1 = Select!(
    Left:int=0
    Right:int=0
    Up:int=-1
    Down:int=1
    Next:int=0
    Previous:int=0
    Invalid:int=0
    Index:EUINavigation Enum=Get Direction() -> (Direction:EUINavigation Enum)
   ) -> (ReturnValue:int)
   n2 = int + int(A=Get TargetRowIndex() -> (TargetRowIndex:int=0), B=@n1)
   n3 = Get Entries() -> (Entries:Array of User Widget Object References)
   > n4 = Branch(Condition=AND Boolean(A=Is Valid Index(TargetArray=@n3, IndexToTest=@n2), B="Not Equal (Integer)"(A=@n1)))
   > n5 = Cast To WBP_UIKit_VariantRow_Wrapper(Object:Object Reference="Get (a copy)"(Array=Get Entries(), Dimension 1=@n2))
   > SetNavigationRuleExplicit!(
    execute
    self:Widget Object Reference=Get TargetRow() -> (TargetRow:User Widget Object Reference)
    Direction:EUINavigation Enum=Get Direction()
    InWidget:Widget Object Reference=GetDesiredFocusTarget(self=@n5)
   ) -> (then)
   SetNavigationRuleBase!(
    execute=@n4.else
    self:Widget Object Reference=Get TargetRow()
    Direction:EUINavigation Enum=Get Direction()
    Rule:EUINavigationRule Enum=Escape
   ) -> (then)
  ]}
 SetMobileTabSizing {nodes [
   SetMobileTabSizing() {is=title K2Node_FunctionEntry}
   > Branch(Condition=IsMobileGame())
   n1 = Get Tabs_PrimaryFilter()
   > Set IconSize!(execute, IconSize:Vector 2D Structure=(X=48,Y=48), self:WBP UIKit Tabs Base Object Reference=@n1) -> (then, Output_Get:Vector 2D Structure)
   > Set TabGroupPadding!(execute, TabGroupPadding:Margin Structure=Make Margin(Left=3, Top=3, Right=3, Bottom=3), self:WBP UIKit Tabs Base Object Reference=@n1) -> (then, Output_Get:Margin Structure)
   n2 = Get Tabs_SecondaryFilter()
   > Set IconSize(IconSize=(X=48,Y=48), self=@n2)
   > Set TabGroupPadding(TabGroupPadding=Make Margin(Left=3, Top=3, Right=3, Bottom=3), self=@n2)
  ]}
 SetTabgroupFromFilters {nodes [
   n1 = SetTabgroupFromFilters!() -> (
    then
    FilterVMs:Array of Object References
    TabGroup:WBP UIKit Tabs Object Reference
    TabGroupHeader:Common Text Object Reference
    OnTabSelected:"Delegate (by ref)"
   ) {is=title K2Node_FunctionEntry}
   > Unbind all Events from On Tab Selected(self=@n1.TabGroup)
   > RemoveAllTabs(self:Common Tab List Widget Base Object Reference=Get TabGroup() -> (TabGroup:WBP UIKit Tabs Object Reference))
   n2 = Get FilterVMs() -> (FilterVMs:Array of Object References)
   n3 = Select!(
    Option 0:ESlateVisibility Enum=Collapsed
    Option 1:ESlateVisibility Enum=SelfHitTestInvisible
    Index:bool="integer > integer"(A=Length(TargetArray:Array of Object References=@n2), B=1)
   ) -> (ReturnValue:ESlateVisibility Enum)
   > Set Visibility(Visibility=@n3, self=Get TabGroup())
   > n4 = Branch(Condition=Get ShowTabHeaders() -> (ShowTabHeaders:bool))
   > n5 = Set Visibility(Visibility=@n3, self=Get TabGroupHeader() -> (TabGroupHeader:Common Text Object Reference))
   n6 = Sequence(execute=[@n5.then, @n4.else])
   > n7 = For Each Loop(Array:Array of Object References=Get FilterVMs()) -> (Array Element:Object Reference)
   > n8 = Construct Object from Class(Class=/Script/FortniteUI.UIKitTabGroupButtonViewModel) -> (ReturnValue:UI Kit Tab Group Button View Model Object Reference)
   > Set LoopTabVM(LoopTabVM:UI Kit Tab Group Button View Model Object Reference=@n8) -> (Output_Get:UI Kit Tab Group Button View Model Object Reference)
   > n9 = Cast To FortVariantSelectorFilterVM!(execute, Object=@n7) -> (then, CastFailed, AsFort Variant Selector Filter VM:Fort Variant Selector Filter VM Object Reference) {is=title K2Node_DynamicCast}
   n10 = Get MainFilter!(self:Fort Variant Selector Filter VM Object Reference=@n9) -> (MainFilter:Fort Variant Selector Sub Filter VM Object Reference)
   > n11 = Set Icon!(
    execute
    Icon:Texture 2D Object Reference=GetIcon!(self, Asset:Soft Object Reference=Get Icon!(self:Fort Variant Selector Sub Filter VM Object Reference=@n10) -> (Icon:Soft Object Reference)) -> (Texture:Texture 2D Object Reference) {is=title K2Node_CallFunction}
    self:UI Kit Selectable Button View Model Object Reference=Get LoopTabVM() -> (LoopTabVM:UI Kit Tab Group Button View Model Object Reference)
   ) -> (then, Output_Get:Texture 2D Object Reference)
   Add(
    execute=[@n13.then, @n11.then]
    TargetArray:Array of UI Kit Tab Group Button View Model Object References=Get TabGroupButtonsViewModels() -> (TabGroupButtonsViewModels:Array of UI Kit Tab Group Button View Model Object References)
    NewItem:"UI Kit Tab Group Button View Model Object Reference (by ref)"=Get LoopTabVM()
   ) {is=title K2Node_CallArrayFunction}
   n12 = Cast To FortVariantSelectorSubFilterVM!(execute=@n9.CastFailed, Object=@n7) -> (then, CastFailed, AsFort Variant Selector Sub Filter VM:Fort Variant Selector Sub Filter VM Object Reference) {is=title K2Node_DynamicCast}
   > n13 = Set Icon(Icon=GetIcon(Asset=Get Icon(self=@n12)), self=Get LoopTabVM())
   n14 = Construct Object from Class(execute=@n6.then_1, Class=/Script/FortniteUI.UIKitTabGroupViewModel) -> (ReturnValue:UI Kit Tab Group View Model Object Reference)
   > Set Tabs!(
    execute
    Tabs:Array of UI Kit Tab Group Button View Model Object References=Get TabGroupButtonsViewModels()
    self:UI Kit Tab Group View Model Object Reference=@n14
   ) -> (then, Output_Get:Array of UI Kit Tab Group Button View Model Object References)
   > SetUIKitTabGroupViewModel!(execute, self:WBP UIKit Tabs Object Reference=Get TabGroup(), Viewmodel:UI Kit Tab Group View Model Object Reference=@n14) -> (then)
   > Unbind all Events from On Tab Selected(self=Get TabGroup())
   > Bind Event to On Tab Selected(self:Common Tab List Widget Base Object Reference=Get TabGroup(), Delegate=Get OnTabSelected() -> (OnTabSelected:"Delegate (by ref)")) {is=title K2Node_AddDelegate}
  ]}
 SetupLayout {nodes [
   SetupLayout!() -> (then, UseCustomizeLayout:bool=~) {is=title K2Node_FunctionEntry}
   > n1 = Set ButtonClass!(
    execute
    ButtonClass:WBP UIKit Button Generic Base Class Reference=Select(
     Option 0:WBP UIKit Button Generic Base Class Reference=/FortUILibrary/UIKit/Buttons/WBP_UIKit_Button_Tab_Enhanced.WBP_UIKit_Button_Tab_Enhanced_C
     Option 1:WBP UIKit Button Generic Base Class Reference=&/WBP_UIKit_Button_Tab_Customize.WBP_UIKit_Button_Tab_Customize_C
     Index=Get UseCustomizeLayout() -> (UseCustomizeLayout:bool)
    ) -> (ReturnValue:WBP UIKit Button Generic Base Class Reference)
    self:WBP UIKit Tabs Base Object Reference=Get Tabs_PrimaryFilter()
   ) -> (then, Output_Get:WBP UIKit Button Generic Base Class Reference)
   > Set ButtonClass(ButtonClass=@n1, self=Get Tabs_SecondaryFilter())
   > n2 = Set IconSize(
    IconSize=Select(Option 0:Vector 2D Structure=(X=32,Y=32), Option 1:Vector 2D Structure=(X=64,Y=64), Index=Get UseCustomizeLayout()) -> (ReturnValue:Vector 2D Structure)
    self=Get Tabs_PrimaryFilter()
   )
   > Set IconSize(IconSize=@n2, self=Get Tabs_SecondaryFilter())
   > n3 = Set Orientation!(
    execute
    Orientation:EOrientation Enum=Select(Option 0:EOrientation Enum=Orient_Vertical, Option 1:EOrientation Enum=Orient_Horizontal, Index=Get UseCustomizeLayout()) -> (ReturnValue:EOrientation Enum)
    self:Stack Box Object Reference=Get RootStackBox() -> (RootStackBox:Stack Box Object Reference)
   ) -> (then, Output_Get:EOrientation Enum=Orient_Horizontal)
   > n4 = Set Orientation(Orientation=@n3, self=Get RootStackBox()) -> (Output_Get=Orient_Horizontal)
   > Set Orientation(Orientation=@n4, self=Get TabsStackBox() -> (TabsStackBox:Stack Box Object Reference)) -> (Output_Get=Orient_Horizontal)
   > n5 = Set IsVertical!(
    execute
    IsVertical:bool=Select(Option 0:bool=false, Option 1:bool=true, Index=Get UseCustomizeLayout()) -> (ReturnValue:bool)
    self:WBP UIKit Tabs Base Object Reference=Get Tabs_SecondaryFilter()
   ) -> (then, Output_Get:bool)
   > Set IsVertical(IsVertical=@n5, self=Get Tabs_PrimaryFilter())
   > n6 = Set ShowTabHeaders(ShowTabHeaders:bool=Select(Option 0=Get ShowTabHeaders(), Option 1=false, Index=Get UseCustomizeLayout())) -> (Output_Get:bool)
   > n7 = Set Visibility(
    Visibility=Select(Option 0:ESlateVisibility Enum=Collapsed, Option 1:ESlateVisibility Enum=SelfHitTestInvisible, Index=@n6) -> (ReturnValue:ESlateVisibility Enum)
    self=Get Text_PrimaryFilter() -> (Text_PrimaryFilter:Common Text Object Reference)
   )
   > Set Visibility(Visibility=@n7, self=Get Text_SecondaryFilter() -> (Text_SecondaryFilter:Common Text Object Reference))
   > n8 = Set TabGroupVerticalAlignment!(
    execute
    TabGroupVerticalAlignment:EVerticalAlignment Enum=Select(Option 0:EVerticalAlignment Enum=VAlign_Fill, Option 1:EVerticalAlignment Enum=VAlign_Top, Index=Get UseCustomizeLayout()) -> (ReturnValue:EVerticalAlignment Enum)
    self:WBP UIKit Tabs Base Object Reference=Get Tabs_PrimaryFilter()
   ) -> (then, Output_Get:EVerticalAlignment Enum=VAlign_Fill)
   > Set TabGroupVerticalAlignment(TabGroupVerticalAlignment=@n8, self=Get Tabs_SecondaryFilter()) -> (Output_Get=VAlign_Fill)
   n9 = Get UseCustomizeLayout()
   > Set!(
    execute
    Padding_Left:float=0
    Padding_Top:float=0
    Padding_Right:float=SelectFloat(A=16, B=0, bPickA=@n9)
    Padding_Bottom:float=0
    self:Stack Box Slot Object Reference=SlotAsStackBoxSlot!(Widget:Widget Object Reference=Get Tabs_PrimaryFilter()) -> (ReturnValue:Stack Box Slot Object Reference)
   ) -> (then, Output_Get:Margin Structure)
   > Set(
    Padding_Left=0
    Padding_Top=0
    Padding_Right=SelectFloat(A=32, B=0, bPickA=@n9)
    Padding_Bottom=0
    self=SlotAsStackBoxSlot(Widget=Get Tabs_SecondaryFilter())
   )
   > Set VerticalAlignment!(
    execute
    VerticalAlignment:EVerticalAlignment Enum=Select(Option 0=VAlign_Fill, Option 1=VAlign_Center, Index=Get UseCustomizeLayout())
    self:Stack Box Slot Object Reference=SlotAsStackBoxSlot(Widget=Get Switcher_Content())
   ) -> (then, Output_Get:EVerticalAlignment Enum=VAlign_Fill)
   n10 = Select(Option 0:ESlateVisibility Enum=SelfHitTestInvisible, Option 1:ESlateVisibility Enum=Collapsed, Index=Get UseCustomizeLayout()) -> (ReturnValue:ESlateVisibility Enum)
   > SetVisibility(
    self=Get ContentBackplate!(self:WBP UIKit Tabs Base Object Reference=Get Tabs_SecondaryFilter()) -> (ContentBackplate:WBP UIKit Backplate Base Object Reference)
    InVisibility=@n10
   )
   > SetVisibility(self=Get ContentBackplate(self=Get Tabs_PrimaryFilter()), InVisibility=@n10)
   > Set TabSpacing!(
    execute
    TabSpacing:double=SelectFloat(A=16, B=8, bPickA=Get UseCustomizeLayout())
    self:WBP UIKit Tabs Base Object Reference=Get Tabs_SecondaryFilter()
   ) -> (then, Output_Get:double)
   > Set TabSpacing(TabSpacing=SelectFloat(A=16, B=8, bPickA=Get UseCustomizeLayout()), self=Get Tabs_PrimaryFilter())
  ]}
 TryFocusToEquipped {nodes [
   TryFocusToEquipped() {is=title K2Node_FunctionEntry}
   n1 = GetAllEntries(self=Get EntryBox_ChannelRows())
   > Branch(Condition=Is Valid Index(TargetArray=@n1))
   > n2 = Cast To WBP_UIKit_VariantRow_Wrapper(Object="Get (a copy)"(Array=@n1))
   > SelectEquippedFocusTarget(self:WBP UIKit Variant Row Wrapper Object Reference=@n2)
   SetFocus(execute=@n2.CastFailed, self=GetDesiredFocusTarget!(self:Self Object Reference) -> (ReturnValue:Widget Object Reference))
  ]}
 VM_PrimaryFiltersUpdated {nodes [
   n1 = VM_PrimaryFiltersUpdated!() -> (then, FilterVMs:Array of Fort Variant Selector Filter VM Object References) {is=title K2Node_FunctionEntry}
   > SetTabgroupFromFilters!(
    execute
    self
    FilterVMs:Array of Object References=@n1
    TabGroup:WBP UIKit Tabs Object Reference=Get Tabs_PrimaryFilter()
    TabGroupHeader:Common Text Object Reference=Get Text_PrimaryFilter()
    OnTabSelected:"Delegate (by ref)"=Create Event()
   ) -> (then) {is=title K2Node_CallFunction}
   n2 = Get Tabs_PrimaryFilter()
   n3 = GetTabIndexByID(self=@n2, TabNameID=0) -> (ReturnValue=0)
   n4 = Get PrimaryFilters(self=Get FortVariantSelectorVM())
   > n5 = Branch(Condition=Is Valid Index(TargetArray:Array of Fort Variant Selector Filter VM Object References=@n4, IndexToTest=@n3))
   n6 = Get FortVariantSelectorVM()
   > n7 = SetCurrentPrimaryFilter(
    self=@n6
    InCurrentPrimaryFilter="Get (a copy)"(Array:Array of Fort Variant Selector Filter VM Object References=@n4, Dimension 1=@n3) -> (Output:Fort Variant Selector Filter VM Object Reference)
   )
   SelectTabByIndex(
    execute=[@n7.then, @n5.else]
    self=@n2
    Index=Get DefaultTabIndex!(self:Fort Variant Selector VM Object Reference=@n6) -> (DefaultTabIndex:int=0)
    bSuppressClickFeedback=true
   )
  ]}
 VM_SecondaryFiltersUpdated {nodes [
   n1 = VM_SecondaryFiltersUpdated!() -> (then, SubFilterVMs:Array of Fort Variant Selector Sub Filter VM Object References) {is=title K2Node_FunctionEntry}
   > SetTabgroupFromFilters(FilterVMs=@n1, TabGroup=Get Tabs_SecondaryFilter(), TabGroupHeader=Get Text_SecondaryFilter(), OnTabSelected=Create Event())
   n2 = Get Tabs_SecondaryFilter()
   n3 = GetTabIndexByID(self=@n2, TabNameID=0) -> (ReturnValue=0)
   n4 = Get SecondaryFilters(self=Get FortVariantSelectorVM())
   > n5 = Branch(Condition=Is Valid Index(TargetArray:Array of Fort Variant Selector Sub Filter VM Object References=@n4, IndexToTest=@n3))
   > n6 = SetCurrentSecondaryFilter(
    self=Get FortVariantSelectorVM()
    InCurrentSecondaryFilter="Get (a copy)"(Array:Array of Fort Variant Selector Sub Filter VM Object References=@n4, Dimension 1=@n3) -> (Output:Fort Variant Selector Sub Filter VM Object Reference)
   )
   SelectTabByIndex(execute=[@n6.then, @n5.else], self=@n2, Index=0, bSuppressClickFeedback=true)
  ]}
 VM_SelectedItemUpdated {nodes [
   n1 = VM_SelectedItemUpdated!() -> (then, SelectedItem:Fort Item VM Object Reference) {is=title K2Node_FunctionEntry}
   > n2 = Get()
   > Unbind all Events from On Selected Item Set Event!(execute:Exec, self:Fort Variant Selector VM Object Reference=@n2) -> (then:Exec) {is=title K2Node_ClearDelegate}
   > Bind Event to On Selected Item Set Event(self:Fort Variant Selector VM Object Reference=@n2, Delegate=Create Event()) {is=title K2Node_AddDelegate}
   > Set Selected Item Async!(execute, self:Fort Variant Selector VM Object Reference=@n2, InSelectedItem:Fort Item VM Object Reference=@n1) -> (then) {is=title K2Node_CallFunction}
  ]}
 VM_VariantSetsUpdated {nodes [
   n1 = VM_VariantSetsUpdated!() -> (then, VariantSetVMs:Array of Fort Variant Set VM Object References) {is=title K2Node_FunctionEntry}
   > Reset!(execute, self:Dynamic Entry Box Object Reference=Get EntryBox_ChannelRows(), bDeleteWidgets:bool=true) -> (then)
   > n2 = For Each Loop(Array:Array of Fort Variant Set VM Object References=@n1) -> (Array Element:Fort Variant Set VM Object Reference)
   SetAllRowsNavigation(execute=@n2.Completed) {is=title K2Node_CallFunction}
   > SetActiveWidget(self=Get Switcher_Content(), Widget=Get ScrollBox_Content() -> (ScrollBox_Content:Scroll Box Object Reference))
   n3 = Get CommonInputSubsystem!(PlayerController:Player Controller Object Reference=GetOwningPlayer()) -> (ReturnValue:Common Input Object Reference) {is=title K2Node_GetSubsystemFromPC}
   > Branch(Condition="Equal (Enum)"(A=GetCurrentInputType(self=@n3), B=Gamepad))
   > TryFocusToEquipped() {is=title K2Node_CallFunction}
   n4 = Create Entry!(execute=@n2.LoopBody, self:Dynamic Entry Box Object Reference=Get EntryBox_ChannelRows()) -> (then, ReturnValue:User Widget Object Reference)
   > n5 = Cast To WBP_UIKit_VariantRow_Wrapper(Object=@n4)
   > Unbind all Events from On Conflicts Updated!(execute:Exec, self:WBP UIKit Variant Row Wrapper Object Reference=@n5) -> (then:Exec) {is=title K2Node_ClearDelegate}
   > Bind Event to On Conflicts Updated(self:WBP UIKit Variant Row Wrapper Object Reference=@n5, Delegate=Create Event()) {is=title K2Node_AddDelegate}
   > Unbind all Events from On Variant Item Engaged!(execute:Exec, self:WBP UIKit Variant Row Wrapper Object Reference=@n5) -> (then:Exec) {is=title K2Node_ClearDelegate}
   > Bind Event to On Variant Item Engaged(self:WBP UIKit Variant Row Wrapper Object Reference=@n5, Delegate=Create Event()) {is=title K2Node_AddDelegate}
   > Unbind all Events from On Variant Item Previewed!(execute:Exec, self:WBP UIKit Variant Row Wrapper Object Reference=@n5) -> (then:Exec) {is=title K2Node_ClearDelegate}
   > Bind Event to On Variant Item Previewed(self:WBP UIKit Variant Row Wrapper Object Reference=@n5, Delegate=Create Event()) {is=title K2Node_AddDelegate}
   > Set Additional Arguments!(
    execute
    Additional Arguments=Get Additional Arguments() -> (Additional Arguments:Map of Names to Booleans)
    self:WBP UIKit Variant Row Wrapper Object Reference=@n5
   ) -> (then, Output_Get)
   > Set TargetStyleTag!(
    execute
    TargetStyleTag:Gameplay Tag Structure=Get VariantChannelStyleTag() -> (VariantChannelStyleTag:Gameplay Tag Structure)
    self:WBP UIKit Variant Row Wrapper Object Reference=@n5
   ) -> (then, Output_Get:Gameplay Tag Structure)
   > Set FortVariantSelectorVM!(
    execute
    FortVariantSelectorVM:Fort Variant Selector VM Object Reference=Get FortVariantSelectorVM()
    self:WBP UIKit Variant Row Wrapper Object Reference=@n5
   ) -> (then, Output_Get:Fort Variant Selector VM Object Reference)
   > Set FortVariantSetVM!(execute, FortVariantSetVM:Fort Variant Set VM Object Reference=@n2, self:WBP UIKit Variant Row Wrapper Object Reference=@n5) -> (then, Output_Get:Fort Variant Set VM Object Reference)
  ]}
]
variables [
 Additional Arguments {type="map<boolean>", default {}}
 DefaultFilterIcon {type=Texture2D, default=/Game/UI/Foundation/Textures/Icons/Locker/T_Icon_Spray_Adjustments_Tab_128}
 DisplayUnownedVariants {type=boolean}
 OnConflictTextUpdated {}
 OnFilterChanged {}
 OnRotationRequested {}
 OnVariantItemEngaged {}
 OnVariantItemPreviewed {}
 ShowAgeGateWarnings {type=boolean}
 ShowTabHeaders {type=boolean, default=true}
 ShowUnownedVariantsAsLocked {type=boolean, default=true}
 Use Horizontal Layout {type=boolean}
 VariantChannelStyleTag {type=GameplayTag, default {tagName=None}}
 VariantPropertyFilterTag {type=GameplayTag, default {tagName=None}}
]
widgets {
 info {inheritedWidgetCount=0, parentClass=&UserWidget, rootWidgetClass=&StackBox, widgetCount=10}
 widgets [
  RootStackBox {widgetClassPath=&StackBox, bIsVariable=true, widget=RootStackBox, properties {orientation=Orient_Vertical}}
  TabsStackBox {
   widgetClassPath=&StackBox
   bIsVariable=true
   parent=RootStackBox
   widget=TabsStackBox
   properties {bOverride_Cursor=true, orientation=Orient_Vertical}
   slot {class=StackBoxSlot, properties {padding=<ltrb 0 0 0 16>}}
  }
  Text_PrimaryFilter {
   widgetClassPath=&CommonTextBlock
   bIsVariable=true
   parent=TabsStackBox
   widget=Text_PrimaryFilter
   properties {
    mobileFontSizeMultiplier=1.3
    scrollStyle=&TextScrollStyle-Base_C
    text=Customize Your Hero
    textTransformPolicy=ToUpper
    font {
     bForceMonospaced=false
     bMaterialIsStencil=false
     fontMaterial=null
     fontObject=&HeadingNow
     letterSpacing=0
     monospacedWidth=1
     size=15.75
     skewAmount=0
     typefaceFontName=75 Medium
     outlineSettings {
      bApplyOutlineToDropShadows=false
      bMiteredCorners=false
      bSeparateFillAlpha=false
      outlineMaterial=null
      outlineSize=0
      outlineColor=<rgba 0 0 0 1>
     }
    }
    shadowOffset=<0 0>
   }
   slot {class=StackBoxSlot, properties {padding=<ltrb 0 0 0 8>}}
  }
  Tabs_PrimaryFilter {
   widgetClassPath=/FortUILibrary/UIKit/Tabs/WBP_UIKit_Tabs
   bIsVariable=true
   parent=TabsStackBox
   widget=Tabs_PrimaryFilter
   properties {
    backplateCornerRadius=Interactive Component
    buttonClass=&WBP_UIKit_Button_Tab_Enhanced_C
    inputActions=Triggers
    tabSpacing=8
    useBackplate=true
   }
   slot {class=StackBoxSlot, properties {padding=<ltrb 0 0 0 16>}}
  }
  Text_SecondaryFilter {
   widgetClassPath=&CommonTextBlock
   bIsVariable=true
   parent=TabsStackBox
   widget=Text_SecondaryFilter
   properties {
    mobileFontSizeMultiplier=1.3
    scrollStyle=&TextScrollStyle-Base_C
    text=Outfit - All
    textTransformPolicy=ToUpper
    font {
     bForceMonospaced=false
     bMaterialIsStencil=false
     fontMaterial=null
     fontObject=&HeadingNow
     letterSpacing=0
     monospacedWidth=1
     size=15.75
     skewAmount=0
     typefaceFontName=75 Medium
     outlineSettings {
      bApplyOutlineToDropShadows=false
      bMiteredCorners=false
      bSeparateFillAlpha=false
      outlineMaterial=null
      outlineSize=0
      outlineColor=<rgba 0 0 0 1>
     }
    }
    shadowOffset=<0 0>
   }
   slot {class=StackBoxSlot, properties {padding=<ltrb 0 8 0 8>}}
  }
  Tabs_SecondaryFilter {
   widgetClassPath=&WBP_UIKit_Tabs
   bIsVariable=true
   parent=TabsStackBox
   widget=Tabs_SecondaryFilter
   properties {
    backplateCornerRadius=Interactive Component
    buttonClass=&WBP_UIKit_Button_Tab_Enhanced_C
    tabSpacing=8
    useBackplate=true
   }
   slot {class=StackBoxSlot, properties {padding=<ltrb 0 0 0 16>}}
  }
  Switcher_Content {
   widgetClassPath=&CommonVisibilitySwitcher
   bIsVariable=true
   parent=RootStackBox
   widget=Switcher_Content
   slot {class=StackBoxSlot, properties {size {sizeRule=Fill, value=1}}}
  }
  ScrollBox_Content {
   widgetClassPath=&ScrollBox
   bIsVariable=true
   parent=Switcher_Content
   widget=ScrollBox_Content
   properties {
    allowOverscroll=false
    navigationScrollPadding=55
    scrollWhenFocusChanges=AnimatedScroll
    wheelScrollMultiplier=2
    scrollbarThickness=<12 12>
    widgetBarStyle {
     thickness=16
     draggedThumbImage {
      bIsDynamicallyLoaded=false
      drawAs=RoundedBox
      imageType=FullColor
      mirroring=NoMirror
      resourceName=../../../Engine/Content/Slate/Common/Scrollbar_Thumb.png
      resourceObject=null
      tiling=NoTile
      imageSize=<12 12>
      margin=<ltrb 0.25 0.25 0.25 0.25>
      outlineSettings {
       bUseBrushTransparency=false
       roundingType=FixedRadius
       width=2
       color {colorUseRule=UseColor_Specified, specifiedColor=<rgba 0 0 0 0>}
       cornerRadii=<6 6 6 6>
      }
      tintColor {colorUseRule=UseColor_Specified, specifiedColor=<rgba 1 1 1 1>}
      uVRegion {bIsValid=false, max=<0 0>, min=<0 0>}
     }
     horizontalBackgroundImage {
      bIsDynamicallyLoaded=false
      drawAs=NoDrawType
      imageType=NoImage
      mirroring=NoMirror
      resourceName=None
      resourceObject=null
      tiling=NoTile
      imageSize=<0 0>
      margin=<ltrb 0 0 0 0>
      outlineSettings {
       bUseBrushTransparency=false
       roundingType=HalfHeightRadius
       width=0
       color {colorUseRule=UseColor_Specified, specifiedColor=<rgba 0 0 0 0>}
       cornerRadii=<0 0 0 0>
      }
      tintColor {colorUseRule=UseColor_Specified, specifiedColor=<rgba 1 1 1 1>}
      uVRegion {bIsValid=false, max=<0 0>, min=<0 0>}
     }
     horizontalBottomSlotImage {
      bIsDynamicallyLoaded=false
      drawAs=NoDrawType
      imageType=FullColor
      mirroring=NoMirror
      resourceName=../../../Engine/Content/Slate/Common/Scrollbar_Background_Horizontal.png
      resourceObject=null
      tiling=NoTile
      imageSize=<8 8>
      margin=<ltrb 0 0 0 0>
      outlineSettings {
       bUseBrushTransparency=false
       roundingType=HalfHeightRadius
       width=0
       color {colorUseRule=UseColor_Specified, specifiedColor=<rgba 0 0 0 0>}
       cornerRadii=<0 0 0 0>
      }
      tintColor {colorUseRule=UseColor_Specified, specifiedColor=<rgba 1 1 1 1>}
      uVRegion {bIsValid=false, max=<0 0>, min=<0 0>}
     }
     horizontalTopSlotImage {
      bIsDynamicallyLoaded=false
      drawAs=NoDrawType
      imageType=FullColor
      mirroring=NoMirror
      resourceName=../../../Engine/Content/Slate/Common/Scrollbar_Background_Horizontal.png
      resourceObject=null
      tiling=NoTile
      imageSize=<8 8>
      margin=<ltrb 0 0 0 0>
      outlineSettings {
       bUseBrushTransparency=false
       roundingType=HalfHeightRadius
       width=0
       color {colorUseRule=UseColor_Specified, specifiedColor=<rgba 0 0 0 0>}
       cornerRadii=<0 0 0 0>
      }
      tintColor {colorUseRule=UseColor_Specified, specifiedColor=<rgba 1 1 1 1>}
      uVRegion {bIsValid=false, max=<0 0>, min=<0 0>}
     }
     hoveredThumbImage {
      bIsDynamicallyLoaded=false
      drawAs=RoundedBox
      imageType=FullColor
      mirroring=NoMirror
      resourceName=../../../Engine/Content/Slate/Common/Scrollbar_Thumb.png
      resourceObject=null
      tiling=NoTile
      imageSize=<12 12>
      margin=<ltrb 0.25 0.25 0.25 0.25>
      outlineSettings {
       bUseBrushTransparency=false
       roundingType=FixedRadius
       width=2
       color {colorUseRule=UseColor_Specified, specifiedColor=<rgba 0 0 0 0>}
       cornerRadii=<6 6 6 6>
      }
      tintColor {colorUseRule=UseColor_Specified, specifiedColor=<rgba 1 1 1 1>}
      uVRegion {bIsValid=false, max=<0 0>, min=<0 0>}
     }
     normalThumbImage {
      bIsDynamicallyLoaded=false
      drawAs=RoundedBox
      imageType=FullColor
      mirroring=NoMirror
      resourceName=../../../Engine/Content/Slate/Common/Scrollbar_Thumb.png
      resourceObject=null
      tiling=NoTile
      imageSize=<12 12>
      margin=<ltrb 0.25 0.25 0.25 0.25>
      outlineSettings {
       bUseBrushTransparency=false
       roundingType=FixedRadius
       width=2
       color {colorUseRule=UseColor_Specified, specifiedColor=<rgba 0 0 0 0>}
       cornerRadii=<6 6 6 6>
      }
      tintColor {colorUseRule=UseColor_Specified, specifiedColor=<rgba 1 1 1 0.749>}
      uVRegion {bIsValid=false, max=<0 0>, min=<0 0>}
     }
     verticalBackgroundImage {
      bIsDynamicallyLoaded=false
      drawAs=Box
      imageType=FullColor
      mirroring=NoMirror
      resourceName=None
      resourceObject=/FortUILibrary/UIKit/ScrollBar/MaterialInstances/MI_UIKit_ScrollbarBackground
      tiling=NoTile
      imageSize=<12 12>
      margin=<ltrb 0.5 0.5 0.5 0.5>
      outlineSettings {
       bUseBrushTransparency=false
       roundingType=HalfHeightRadius
       width=0
       color {colorUseRule=UseColor_Specified, specifiedColor=<rgba 0 0 0 0>}
       cornerRadii=<0 0 0 0>
      }
      tintColor {colorUseRule=UseColor_Specified, specifiedColor=<rgba 0.521 0.521 0.521 0.502>}
      uVRegion {bIsValid=false, max=<0 0>, min=<0 0>}
     }
     verticalBottomSlotImage {
      bIsDynamicallyLoaded=false
      drawAs=NoDrawType
      imageType=FullColor
      mirroring=NoMirror
      resourceName=../../../Engine/Content/Slate/Common/Scrollbar_Background_Vertical.png
      resourceObject=null
      tiling=NoTile
      imageSize=<8 8>
      margin=<ltrb 0 0 0 0>
      outlineSettings {
       bUseBrushTransparency=false
       roundingType=HalfHeightRadius
       width=0
       color {colorUseRule=UseColor_Specified, specifiedColor=<rgba 0 0 0 0>}
       cornerRadii=<0 0 0 0>
      }
      tintColor {colorUseRule=UseColor_Specified, specifiedColor=<rgba 1 1 1 1>}
      uVRegion {bIsValid=false, max=<0 0>, min=<0 0>}
     }
     verticalTopSlotImage {
      bIsDynamicallyLoaded=false
      drawAs=NoDrawType
      imageType=FullColor
      mirroring=NoMirror
      resourceName=../../../Engine/Content/Slate/Common/Scrollbar_Background_Vertical.png
      resourceObject=null
      tiling=NoTile
      imageSize=<8 8>
      margin=<ltrb 0 0 0 0>
      outlineSettings {
       bUseBrushTransparency=false
       roundingType=HalfHeightRadius
       width=0
       color {colorUseRule=UseColor_Specified, specifiedColor=<rgba 0 0 0 0>}
       cornerRadii=<0 0 0 0>
      }
      tintColor {colorUseRule=UseColor_Specified, specifiedColor=<rgba 1 1 1 1>}
      uVRegion {bIsValid=false, max=<0 0>, min=<0 0>}
     }
    }
    widgetStyle {
     barThickness=12
     bottomShadowBrush {
      bIsDynamicallyLoaded=false
      drawAs=NoDrawType
      imageType=FullColor
      mirroring=NoMirror
      resourceName=../../../Engine/Content/Slate/Common/ScrollBoxShadowBottom.png
      resourceObject=null
      tiling=NoTile
      imageSize=<16 8>
      margin=<ltrb 0.5 0 0.5 1>
      outlineSettings {
       bUseBrushTransparency=false
       roundingType=HalfHeightRadius
       width=0
       color {colorUseRule=UseColor_Specified, specifiedColor=<rgba 0 0 0 0>}
       cornerRadii=<0 0 0 0>
      }
      tintColor {colorUseRule=UseColor_Specified, specifiedColor=<rgba 1 1 1 1>}
      uVRegion {bIsValid=false, max=<0 0>, min=<0 0>}
     }
     horizontalScrolledContentPadding=<ltrb 0 0 1 0>
     leftShadowBrush {
      bIsDynamicallyLoaded=false
      drawAs=Box
      imageType=FullColor
      mirroring=NoMirror
      resourceName=../../../Engine/Content/Slate/Common/ScrollBoxShadowLeft.png
      resourceObject=null
      tiling=NoTile
      imageSize=<8 16>
      margin=<ltrb 1 0.5 0 0.5>
      outlineSettings {
       bUseBrushTransparency=false
       roundingType=HalfHeightRadius
       width=0
       color {colorUseRule=UseColor_Specified, specifiedColor=<rgba 0 0 0 0>}
       cornerRadii=<0 0 0 0>
      }
      tintColor {colorUseRule=UseColor_Specified, specifiedColor=<rgba 1 1 1 1>}
      uVRegion {bIsValid=false, max=<0 0>, min=<0 0>}
     }
     rightShadowBrush {
      bIsDynamicallyLoaded=false
      drawAs=Box
      imageType=FullColor
      mirroring=NoMirror
      resourceName=../../../Engine/Content/Slate/Common/ScrollBoxShadowRight.png
      resourceObject=null
      tiling=NoTile
      imageSize=<8 16>
      margin=<ltrb 0 0.5 1 0.5>
      outlineSettings {
       bUseBrushTransparency=false
       roundingType=HalfHeightRadius
       width=0
       color {colorUseRule=UseColor_Specified, specifiedColor=<rgba 0 0 0 0>}
       cornerRadii=<0 0 0 0>
      }
      tintColor {colorUseRule=UseColor_Specified, specifiedColor=<rgba 1 1 1 1>}
      uVRegion {bIsValid=false, max=<0 0>, min=<0 0>}
     }
     topShadowBrush {
      bIsDynamicallyLoaded=false
      drawAs=NoDrawType
      imageType=FullColor
      mirroring=NoMirror
      resourceName=../../../Engine/Content/Slate/Common/ScrollBoxShadowTop.png
      resourceObject=null
      tiling=NoTile
      imageSize=<16 8>
      margin=<ltrb 0.5 1 0.5 0>
      outlineSettings {
       bUseBrushTransparency=false
       roundingType=HalfHeightRadius
       width=0
       color {colorUseRule=UseColor_Specified, specifiedColor=<rgba 0 0 0 0>}
       cornerRadii=<0 0 0 0>
      }
      tintColor {colorUseRule=UseColor_Specified, specifiedColor=<rgba 1 1 1 1>}
      uVRegion {bIsValid=false, max=<0 0>, min=<0 0>}
     }
     verticalScrolledContentPadding=<ltrb 0 0 0 1>
    }
   }
   slot {class=CommonVisibilitySwitcherSlot, properties {padding=<ltrb -16 0 0 0>}}
  }
  EntryBox_ChannelRows {
   widgetClassPath=&DynamicEntryBox
   bIsVariable=true
   parent=ScrollBox_Content
   widget=EntryBox_ChannelRows
   properties {
    entryBoxType=Vertical
    entryWidgetClass=&/WBP_UIKit_VariantRow_Wrapper.WBP_UIKit_VariantRow_Wrapper_C
    numDesignerPreviewEntries=5
    visibility=Visible
    entrySpacing=<0 16>
   }
   slot {class=ScrollBoxSlot, properties {padding=<ltrb 16 0 8 16>}}
  }
  Throbber_Content {
   widgetClassPath=/FortUILibrary/UIKit/Throbber/WBP_UIKit_Throbber
   bIsVariable=true
   parent=Switcher_Content
   widget=Throbber_Content
   slot {class=CommonVisibilitySwitcherSlot, properties {horizontalAlignment=HAlign_Center, verticalAlignment=VAlign_Center}}
  }
 ]
}
