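# Layout-less base widget for Fortnite cosmetic variant pickers; mandatory and optional getters stand in for subclass-supplied widgets.
# It passes option flags and config to FortVariantSelectorVM, whose VM_ callbacks build primary and secondary filter tabs and one row wrapper per variant set.
# Tab picks set VM filters behind a loading throbber; row engage, preview and conflict events, plus VM rotation requests, are re-broadcast.
# Explicit row navigation and equipped-item focus support gamepads.
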WidgetBlueprint /FortUILibrary/UIKit/VariantSelector/WBP_UIKit_VariantSelector_Base
graphs [
 ApplyVariantOptions {nodes [
   ApplyVariantOptions() {is=title K2Node_FunctionEntry}
   n1 = Get FortVariantSelectorVM() -> (FortVariantSelectorVM:Fort Variant Selector VM Object Reference)
   > Set bAllowDisplayUnownedVariants!(
    execute
    bAllowDisplayUnownedVariants:bool=Get DisplayUnownedVariants() -> (DisplayUnownedVariants:bool)
    self:Fort Variant Selector VM Object Reference=@n1
   ) -> (then, Output_Get:bool)
   > Set bShowUnownedAsLocked!(
    execute
    bShowUnownedAsLocked:bool=Get ShowUnownedVariantsAsLocked() -> (ShowUnownedVariantsAsLocked:bool)
    self:Fort Variant Selector VM Object Reference=@n1
   ) -> (then, Output_Get:bool)
   > Set bShowAgeGateWarnings!(
    execute
    bShowAgeGateWarnings:bool=Get ShowAgeGateWarnings() -> (ShowAgeGateWarnings:bool)
    self:Fort Variant Selector VM Object Reference=@n1
   ) -> (then, Output_Get:bool)
   > Set bForceApplyVariantOverrides!(
    execute
    bForceApplyVariantOverrides:bool=Get ForceApplyVariantOverrides() -> (ForceApplyVariantOverrides:bool)
    self:Fort Variant Selector VM Object Reference=@n1
   ) -> (then, Output_Get:bool)
   > Set bAllowSelectionOfLockedVariants!(
    execute
    bAllowSelectionOfLockedVariants:bool=Get AllowSelectionOfLockedVariants() -> (AllowSelectionOfLockedVariants:bool)
    self:Fort Variant Selector VM Object Reference=@n1
   ) -> (then, Output_Get:bool)
  ]}
 EngageAllSelectedItems {nodes [
   EngageAllSelectedItems() {is=title K2Node_FunctionEntry}
   > n1 = For Each Loop(Array:Array of User Widget Object References=GetBoxEntries!(self) -> (ReturnValue:Array of User Widget Object References)) -> (Array Element:User Widget Object Reference)
   > n2 = Cast To WBP_UIKit_VariantRow_Wrapper!(execute, Object=@n1) -> (then, CastFailed, AsWBP UIKit Variant Row Wrapper:WBP UIKit Variant Row Wrapper Object Reference) {is=title K2Node_DynamicCast}
   > EngageSelectedItem(self:WBP UIKit Variant Row Wrapper Object Reference=@n2)
  ]}
 EventGraph {nodes [
   n1 = HandleOnVariantItemEngaged!() -> (
    OutputDelegate
    then
    VariantItem:Fort Variant Item VM Object Reference
    VariantChannel:Fort Variant Channel VM Object Reference
    AdditionalInfo:String
   ) {is=title K2Node_CustomEvent}
   > UpdateVariantChannelConflicts(self:Fort Variant Selector VM Object Reference=Get FortVariantSelectorVM())
   > Call On Variant Item Engaged!(
    execute:Exec
    self:WBP UIKit Variant Selector Base Object Reference
    VariantItem:Fort Variant Item VM Object Reference=@n1.VariantItem
    VariantChannel:Fort Variant Channel VM Object Reference=@n1.VariantChannel
    AdditionalInfo:String=@n1.AdditionalInfo
   ) -> (then:Exec) {is=title K2Node_CallDelegate}
   n2 = OnSecondaryTabSelected!() -> (OutputDelegate, then, TabID:Name) {is=title K2Node_CustomEvent}
   > Get Valid Secondary Tabs!(In:Exec) -> (IsValid:Exec, NotValid:Exec, UIKitTabs:WBP UIKit Tabs Object Reference) {is=title K2Node_MacroInstance}
   > SetActiveWidget(
    self=GetContentSwitcher_Mandatory!(self) -> (VisibilitySwitcher:Common Visibility Switcher Object Reference)
    Widget=GetContentLoadingWidget_Mandatory!(self) -> (ThrobberWidget:Widget Object Reference)
   )
   n3 = GetSecondaryFilterTabs_Optional!(self) -> (UIKitTabs:WBP UIKit Tabs Object Reference)
   n4 = Get FortVariantSelectorVM()
   > SetCurrentSecondaryFilter!(
    execute
    self:Fort Variant Selector VM Object Reference=Get FortVariantSelectorVM()
    InCurrentSecondaryFilter:Fort Variant Selector Sub Filter VM Object Reference="Get (a copy)"(
     Array:Array of Fort Variant Selector Sub Filter VM Object References=Get SecondaryFilters!(self:Fort Variant Selector VM Object Reference=@n4) -> (SecondaryFilters:Array of Fort Variant Selector Sub Filter VM Object References)
     Dimension 1=GetTabIndexByID!(self:UIKit Tab Group Object Reference=@n3, TabNameID:Name=@n2.TabID) -> (ReturnValue:int=0)
    ) -> (Output:Fort Variant Selector Sub Filter VM Object Reference)
   ) -> (then)
   > Call On Filter Changed!(execute:Exec, self:WBP UIKit Variant Selector Base Object Reference) -> (then:Exec) {is=title K2Node_CallDelegate}
   n5 = HandleOnVariantItemPreviewed!() -> (OutputDelegate, then, VariantItem:Fort Variant Item VM Object Reference) {is=title K2Node_CustomEvent}
   > Call On Variant Item Previewed!(execute:Exec, self:WBP UIKit Variant Selector Base Object Reference, VariantItem:Fort Variant Item VM Object Reference=@n5.VariantItem) -> (then:Exec) {is=title K2Node_CallDelegate}
   n6 = HandleOnSelectedItemSet!() -> (OutputDelegate, then, SelectedItem:Fort Item VM Object Reference) {is=title K2Node_CustomEvent}
   > DismissVariantChannels(self:Fort Variant Selector VM Object Reference=Get FortVariantSelectorVM())
   > SetVariantPropertyQueryFilter!(
    execute
    self:Fort Variant Selector VM Object Reference=Get FortVariantSelectorVM()
    InTag:Gameplay Tag Structure=Get VariantPropertyFilterTag() -> (VariantPropertyFilterTag:Gameplay Tag Structure)
    bForceUpdate:bool=false
   ) -> (then)
   > UpdateVariantChannels(self:Fort Variant Selector VM Object Reference=Get FortVariantSelectorVM())
   > HandleOnConflictsUpdated!(
    execute
    self
    Conflicts:Map of E_VariantConflictType Enums to Texts=GetConflictsFromItem!(
     SelectedItem:Fort Item VM Object Reference=@n6.SelectedItem
     VariantSelectorVM:Fort Variant Selector VM Object Reference=Get FortVariantSelectorVM()
    ) -> (Conflicts:Map of E_VariantConflictType Enums to Texts=NewEnumerator6)
   ) -> (then)
   n7 = HandleOnConflictsUpdated!() -> (OutputDelegate, then, Conflicts:Map of E_VariantConflictType Enums to Texts=NewEnumerator6) {is=title K2Node_CustomEvent}
   n8 = GetPrioritizedConflictFromMap!(Conflicts:Map of E_VariantConflictType Enums to Texts=@n7.Conflicts) -> (ConflictText:Text, ConflictType:E_VariantConflictType Enum=NewEnumerator6)
   > Call On Conflict Text Updated!(
    execute:Exec
    self:WBP UIKit Variant Selector Base Object Reference
    ConflictText:Text=@n8
    ConflictType:E_VariantConflictType Enum=@n8.ConflictType
   ) -> (then:Exec) {is=title K2Node_CallDelegate}
   n9 = OnPrimaryTabSelected!() -> (OutputDelegate, then, TabID:Name) {is=title K2Node_CustomEvent}
   > Get Valid Primary Tabs!(In:Exec) -> (IsValid:Exec, NotValid:Exec, UIKitTabs:WBP UIKit Tabs Object Reference) {is=title K2Node_MacroInstance}
   > SetActiveWidget(self=GetContentSwitcher_Mandatory(), Widget=GetContentLoadingWidget_Mandatory())
   n10 = GetPrimaryFilterTabs_Optional!(self) -> (UIKitTabs:WBP UIKit Tabs Object Reference)
   n11 = Get FortVariantSelectorVM()
   > SetCurrentPrimaryFilter!(
    execute
    self:Fort Variant Selector VM Object Reference=Get FortVariantSelectorVM()
    InCurrentPrimaryFilter:Fort Variant Selector Filter VM Object Reference="Get (a copy)"(
     Array:Array of Fort Variant Selector Filter VM Object References=Get PrimaryFilters!(self:Fort Variant Selector VM Object Reference=@n11) -> (PrimaryFilters:Array of Fort Variant Selector Filter VM Object References)
     Dimension 1=GetTabIndexByID(self=@n10, TabNameID=@n9.TabID) -> (ReturnValue=0)
    ) -> (Output:Fort Variant Selector Filter VM Object Reference)
   ) -> (then)
   > Call On Filter Changed()
   > n12 = Get Valid Secondary Tabs()
   > SelectTabByIndex!(execute, self:UIKit Tab Group Object Reference=@n12, Index:int=0, bSuppressClickFeedback:bool=true) -> (then, ReturnValue:bool)
   n13 = HandleOnRotationRequested!() -> (OutputDelegate, then, RotationOffset:Rotator=~, Snap:bool=~) {is=title K2Node_CustomEvent}
   > Call On Rotation Requested!(execute:Exec, self:WBP UIKit Variant Selector Base Object Reference, RotationOffset:Rotator=@n13.RotationOffset, Snap:bool=@n13.Snap) -> (then:Exec) {is=title K2Node_CallDelegate}
   Event Destruct()
   > n14 = Get Valid Primary Tabs()
   > n15 = Unbind all Events from On Tab Selected!(execute:Exec, self:Common Tab List Widget Base Object Reference=@n14) -> (then:Exec) {is=title K2Node_ClearDelegate}
   n16 = Get Valid Secondary Tabs(In=[@n15.then, @n14.NotValid])
   > n17 = Unbind all Events from On Tab Selected(self=@n16)
   n18 = Create Event()
   Unbind Event from On Input Method Changed(
    execute=[@n17.then, @n16.NotValid]
    self=Get CommonInputSubsystem!(PlayerController:Player Controller Object Reference=GetOwningPlayer()) -> (ReturnValue:Common Input Object Reference) {is=title K2Node_GetSubsystemFromPC}
    Delegate=@n18
   )
   Event Construct()
   > Get!(execute:Exec) -> (then:Exec, else:Exec, FortVariantSelectorVM:Fort Variant Selector VM Object Reference)
   > ApplyVariantOptions() {is=title K2Node_CallFunction}
   n19 = Get FortVariantSelectorVM()
   > SetConfigurationData!(
    execute
    self:Fort Variant Selector VM Object Reference=@n19
    InConfigurationData:Athena Variant Picker Config Object Reference=/Game/Athena/UI/Customization/FortVariantPickerConfig
    bForceUpdate:bool=false
   ) -> (then)
   > Unbind all Events from On Rotation Requested Event!(execute:Exec, self:Fort Variant Selector VM Object Reference=@n19) -> (then:Exec) {is=title K2Node_ClearDelegate}
   > Bind Event to On Rotation Requested Event(self:Fort Variant Selector VM Object Reference=@n19, Delegate=Create Event()) {is=title K2Node_AddDelegate}
   > SetMobileTabSizing()
   > Bind Event to On Input Method Changed(self=Get CommonInputSubsystem(PlayerController=GetOwningPlayer()), Delegate=@n18)
  ]}
 FocusContent {nodes [
   FocusContent!() -> (then, Navigation:EUINavigation Enum) {is=title K2Node_FunctionEntry}
   > Set bIsTabGroupFocused(bIsTabGroupFocused:bool=false) -> (Output_Get:bool)
   > Return Node!(execute, Widget:Widget Object Reference=GetDesiredFocusTarget!(self) -> (ReturnValue:Widget Object Reference))
  ]}
 FocusPrimaryTabs {nodes [
   FocusPrimaryTabs!() -> (then, Navigation:EUINavigation Enum) {is=title K2Node_FunctionEntry}
   > Set bIsTabGroupFocused(bIsTabGroupFocused=true)
   > n1 = Get Valid Primary Tabs()
   > Return Node(Widget=GetTabButtonBaseByID!(
     self:Common Tab List Widget Base Object Reference=@n1
     TabNameID:Name=GetSelectedTabId!(self:Common Tab List Widget Base Object Reference=@n1) -> (ReturnValue:Name=None)
    ) -> (ReturnValue:Common Button Base Object Reference))
  ]}
 FocusSecondaryTabs {nodes [
   FocusSecondaryTabs!() -> (then, Navigation:EUINavigation Enum) {is=title K2Node_FunctionEntry}
   > Set bIsTabGroupFocused(bIsTabGroupFocused=true)
   > n1 = Get Valid Secondary Tabs()
   > Return Node(Widget=GetTabButtonBaseByID(self=@n1, TabNameID=GetSelectedTabId(self=@n1) -> (ReturnValue=None)))
  ]}
 GetBoxEntries {nodes [
   GetBoxEntries() {is=title K2Node_FunctionEntry}
   > Return Node!(
    execute
    ReturnValue:Array of User Widget Object References=GetAllEntries!(self:Dynamic Entry Box Base Object Reference=GetChannelRowsBox_Mandatory!(self) -> (EntryBox:Dynamic Entry Box Object Reference)) -> (ReturnValue:Array of User Widget Object References)
   )
  ]}
 GetChannelRowsBox_Mandatory {nodes [
   GetChannelRowsBox_Mandatory() {is=title K2Node_FunctionEntry}
   > Return Node!(execute, EntryBox:Dynamic Entry Box Object Reference)
  ]}
 GetContentLoadingWidget_Mandatory {nodes [
   GetContentLoadingWidget_Mandatory() {is=title K2Node_FunctionEntry}
   > Return Node!(execute, ThrobberWidget:Widget Object Reference)
  ]}
 GetContentSwitcher_Mandatory {nodes [
   GetContentSwitcher_Mandatory() {is=title K2Node_FunctionEntry}
   > Return Node!(execute, VisibilitySwitcher:Common Visibility Switcher Object Reference)
  ]}
 GetContentWidget_Mandatory {nodes [
   GetContentWidget_Mandatory() {is=title K2Node_FunctionEntry}
   > Return Node!(execute, ContentWidget:Widget Object Reference)
  ]}
 GetDesiredFocusTarget {nodes [
   GetDesiredFocusTarget() {is=title K2Node_FunctionEntry}
   n1 = GetBoxEntries() {is=title K2Node_CallFunction}
   > n2 = Branch(Condition=Is Valid Index!(TargetArray:Array of User Widget Object References=@n1, IndexToTest:int) -> (ReturnValue:bool))
   > n3 = Cast To WBP_UIKit_VariantRow_Wrapper(Object="Get (a copy)"(Array:Array of User Widget Object References=@n1) -> (Output:User Widget Object Reference))
   > Return Node!(
    execute
    ReturnValue:Widget Object Reference=GetDesiredFocusTarget!(self:WBP UIKit Variant Row Wrapper Object Reference=@n3) -> (FocusTarget:Widget Object Reference) {is=title K2Node_CallFunction}
   )
   Return Node(execute=[@n3.CastFailed, @n2.else])
  ]}
 GetIcon {nodes [
   n1 = GetIcon!() -> (then, Asset:Soft Object Reference) {is=title K2Node_FunctionEntry}
   > n2 = LoadAsset_Blocking(Asset=Select(
     Option 0:Soft Object Reference=Get DefaultFilterIcon() -> (DefaultFilterIcon:Texture 2D Soft Object Reference)
     Option 1:Soft Object Reference=@n1
     Index=IsValidSoftObjectReference(SoftObjectReference=@n1)
    ) -> (ReturnValue:Soft Object Reference))
   > n3 = Cast To Texture2D(Object=@n2)
   > n4 = Set TempTexture(TempTexture:Texture 2D Object Reference=@n3) -> (Output_Get:Texture 2D Object Reference)
   Return Node!(execute=[@n8.then, @n4.then], Texture:Texture 2D Object Reference=Get TempTexture() -> (TempTexture:Texture 2D Object Reference))
   n5 = Cast To MaterialInstance!(execute=@n3.CastFailed, Object=@n2) -> (then, CastFailed, AsMaterial Instance:Material Instance Object Reference) {is=title K2Node_DynamicCast}
   n6 = Get TextureParameterValues!(self:Material Instance Object Reference=@n5) -> (TextureParameterValues:Array of Texture Parameter Value Structures)
   > n7 = Cast To Texture2D(Object=Break Texture Parameter Value!(TextureParameterValue:"Texture Parameter Value Structure (by ref)"="Get (a ref)"(Array:Array of Texture Parameter Value Structures=@n6, Dimension 1=0) -> (Output:"Texture Parameter Value Structure (by ref)")) -> (ParameterInfo:Material Parameter Info Structure, ParameterValue:Texture Object Reference) {is=title K2Node_BreakStruct}.ParameterValue)
   > n8 = Set TempTexture(TempTexture=@n7)
  ]}
 GetPrimaryFilterTabs_Optional {nodes [
   GetPrimaryFilterTabs_Optional() {is=title K2Node_FunctionEntry}
   > Return Node!(execute, UIKitTabs:WBP UIKit Tabs Object Reference)
  ]}
 GetPrimaryFilterTextBlock_Optional {nodes [
   GetPrimaryFilterTextBlock_Optional() {is=title K2Node_FunctionEntry}
   > Return Node!(execute, PrimaryFilterTextBlock.:Common Text Object Reference)
  ]}
 GetSecondaryFilterTabs_Optional {nodes [
   GetSecondaryFilterTabs_Optional() {is=title K2Node_FunctionEntry}
   > Return Node!(execute, UIKitTabs:WBP UIKit Tabs Object Reference)
  ]}
 GetSecondaryFilterTextBlock_Optional {nodes [
   GetSecondaryFilterTextBlock_Optional() {is=title K2Node_FunctionEntry}
   > Return Node!(execute, TextBlock:Common Text Object Reference)
  ]}
 GetValidPrimaryTabs {nodes [
   Inputs!() -> (In:Exec)
   n1 = GetPrimaryFilterTabs_Optional() {is=title K2Node_CallFunction}
   > n2 = Is Valid(InputObject=@n1)
   > Outputs!(IsValid:Exec, NotValid:Exec=@n2.Is Not Valid, UIKitTabs:WBP UIKit Tabs Object Reference=@n1)
  ]}
 GetValidSecondaryTabs {nodes [
   Inputs()
   n1 = GetSecondaryFilterTabs_Optional() {is=title K2Node_CallFunction}
   > n2 = Is Valid(InputObject=@n1)
   > Outputs(NotValid=@n2.Is Not Valid, UIKitTabs=@n1)
  ]}
 HandleNextTab_Primary {nodes [
   HandleNextTab_Primary() {is=title K2Node_FunctionEntry}
   > n1 = Get Valid Primary Tabs()
   > SelectNextTab!(execute, self:UIKit Tab Group Object Reference=@n1, bSuppressClickFeedback:bool=true) -> (then)
   > Return Node!(execute, bPassThrough:bool=false)
  ]}
 HandlePrevTab_Primary {nodes [
   HandlePrevTab_Primary() {is=title K2Node_FunctionEntry}
   > n1 = Get Valid Primary Tabs()
   > SelectPreviousTab!(execute, self:UIKit Tab Group Object Reference=@n1, bSuppressClickFeedback:bool=true) -> (then)
   > Return Node(bPassThrough=false)
  ]}
 HandleStwNextTabInput {nodes [
   HandleStwNextTabInput() {is=title K2Node_FunctionEntry}
   > n1 = Get Valid Secondary Tabs()
   > SelectNextTab(self=@n1, bSuppressClickFeedback=true)
   > Return Node!(execute, Pass Through:bool=false)
  ]}
 HandleStwPreviousTabInput {nodes [
   HandleStwPreviousTabInput() {is=title K2Node_FunctionEntry}
   > n1 = Get Valid Secondary Tabs()
   > SelectPreviousTab(self=@n1, bSuppressClickFeedback=true)
   > Return Node(Pass Through=false)
  ]}
 OnConflictTextUpdated {nodes [OnConflictTextUpdated!() -> (then, ConflictText:Text, ConflictType:E_VariantConflictType Enum=NewEnumerator6) {is=title K2Node_FunctionEntry}]}
 OnFilterChanged {nodes [OnFilterChanged() {is=title K2Node_FunctionEntry}]}
 OnInputDeviceChanged {nodes [
   OnInputDeviceChanged!() -> (then, NewInputType:ECommonInputType Enum) {is=title K2Node_FunctionEntry}
   > TryFocusToEquipped()
  ]}
 OnRotationRequested {nodes [OnRotationRequested!() -> (then, RotationOffset:Rotator=~, Snap:bool=~) {is=title K2Node_FunctionEntry}]}
 OnVariantItemEngaged {nodes [OnVariantItemEngaged!() -> (then, VariantItem:Fort Variant Item VM Object Reference, VariantChannel:Fort Variant Channel VM Object Reference, AdditionalInfo:String) {is=title K2Node_FunctionEntry}]}
 OnVariantItemPreviewed {nodes [OnVariantItemPreviewed!() -> (then, VariantItem:Fort Variant Item VM Object Reference) {is=title K2Node_FunctionEntry}]}
 SetAdditionalArguments {nodes [
   n1 = SetAdditionalArguments!() -> (then, Additional Arguments:Map of Names to Booleans) {is=title K2Node_FunctionEntry}
   > Set Additional Arguments(Additional Arguments:Map of Names to Booleans=@n1) -> (Output_Get:Map of Names to Booleans)
  ]}
 SetAllRowsNavigation {nodes [
   SetAllRowsNavigation() {is=title K2Node_FunctionEntry}
   > n1 = Set Entries(Entries:Array of User Widget Object References=GetBoxEntries()) -> (Output_Get:Array of User Widget Object References)
   > n2 = For Each Loop(Array=@n1)
   > SetCustomRowNavigation!(execute, self, TargetRow:User Widget Object Reference=@n2, TargetRowIndex:int=@n2.Array Index, Direction:EUINavigation Enum=Up) -> (then)
   > SetCustomRowNavigation(TargetRow=@n2, TargetRowIndex=@n2.Array Index, Direction=Down)
   Cast To WBP_UIKit_VariantRow_Wrapper(Object:Wildcard)
  ]}
 SetCustomRowNavigation {nodes [
   SetCustomRowNavigation!() -> (then, TargetRow:User Widget Object Reference, TargetRowIndex:int, Direction:EUINavigation Enum) {is=title K2Node_FunctionEntry}
   > Set Entries(Entries=GetBoxEntries())
   n1 = Select!(
    Left:int=0
    Right:int=0
    Up:int=-1
    Down:int=1
    Next:int=0
    Previous:int=0
    Invalid:int=0
    Index:EUINavigation Enum=Get Direction() -> (Direction:EUINavigation Enum)
   ) -> (ReturnValue:int)
   n2 = int + int(A=Get TargetRowIndex() -> (TargetRowIndex:int=0), B=@n1)
   n3 = Get Entries() -> (Entries:Array of User Widget Object References)
   > n4 = Branch(Condition=AND Boolean(A=Is Valid Index(TargetArray=@n3, IndexToTest=@n2), B="Not Equal (Integer)"(A=@n1)))
   > n5 = Cast To WBP_UIKit_VariantRow_Wrapper(Object:Object Reference="Get (a copy)"(Array=Get Entries(), Dimension 1=@n2))
   > SetNavigationRuleExplicit!(
    execute
    self:Widget Object Reference=Get TargetRow() -> (TargetRow:User Widget Object Reference)
    Direction:EUINavigation Enum=Get Direction()
    InWidget:Widget Object Reference=GetDesiredFocusTarget(self=@n5)
   ) -> (then)
   SetNavigationRuleBase!(
    execute=@n4.else
    self:Widget Object Reference=Get TargetRow()
    Direction:EUINavigation Enum=Get Direction()
    Rule:EUINavigationRule Enum=Escape
   ) -> (then)
  ]}
 SetMobileTabSizing {nodes [
   SetMobileTabSizing() {is=title K2Node_FunctionEntry}
   > Branch(Condition=IsMobileGame())
   > n1 = Get Valid Primary Tabs()
   > Set IconSize!(
    execute
    IconSize:Vector 2D Structure=Get MobileTabIconSize() -> (MobileTabIconSize:Vector 2D Structure)
    self:WBP UIKit Tabs Base Object Reference=@n1
   ) -> (then, Output_Get:Vector 2D Structure)
   > n2 = Set TabGroupPadding!(
    execute
    TabGroupPadding:Margin Structure=Get MobileTabPadding() -> (MobileTabPadding:Margin Structure)
    self:WBP UIKit Tabs Base Object Reference=@n1
   ) -> (then, Output_Get:Margin Structure)
   n3 = Get Valid Secondary Tabs(In=[@n2.then, @n1.NotValid])
   > Set IconSize(IconSize=Get MobileTabIconSize(), self=@n3)
   > Set TabGroupPadding(TabGroupPadding=Get MobileTabPadding(), self=@n3)
  ]}
 SetTabgroupFromFilters {nodes [
   n1 = SetTabgroupFromFilters!() -> (
    then
    FilterVMs:Array of Object References
    TabBase:WBP UIKit Tabs Object Reference
    TabGroupHeader:Common Text Object Reference
    OnTabSelected:"Delegate (by ref)"
   ) {is=title K2Node_FunctionEntry}
   > n2 = Is Valid(InputObject=@n1.TabBase)
   > RemoveAllTabs(self:Common Tab List Widget Base Object Reference=Get TabBase() -> (TabBase:WBP UIKit Tabs Object Reference))
   n3 = Get FilterVMs() -> (FilterVMs:Array of Object References)
   n4 = Select!(
    Option 0:ESlateVisibility Enum=Collapsed
    Option 1:ESlateVisibility Enum=SelfHitTestInvisible
    Index:bool="integer > integer"(A=Length(TargetArray:Array of Object References=@n3), B=1)
   ) -> (ReturnValue:ESlateVisibility Enum)
   > SetVisibility(self=Get TabBase(), InVisibility=@n4)
   > n5 = Is Valid(InputObject=Get TabGroupHeader() -> (TabGroupHeader:Common Text Object Reference))
   > n6 = SetVisibility(self=Get TabGroupHeader(), InVisibility=@n4)
   n7 = Sequence!(execute=[@n6.then, @n5.Is Not Valid]) -> (then_0, then_1, then_2:Exec)
   > Unbind all Events from On Tab Selected(self=Get TabBase())
   > Bind Event to On Tab Selected(self:Common Tab List Widget Base Object Reference=Get TabBase(), Delegate=Get OnTabSelected() -> (OnTabSelected:"Delegate (by ref)")) {is=title K2Node_AddDelegate}
   n8 = For Each Loop(Exec=@n7.then_1, Array:Array of Object References=Get FilterVMs()) -> (Array Element:Object Reference)
   > n9 = Construct Object from Class(Class=/Script/FortniteUI.UIKitTabGroupButtonViewModel) -> (ReturnValue:UI Kit Tab Group Button View Model Object Reference)
   > Set LoopTabVM(LoopTabVM:UI Kit Tab Group Button View Model Object Reference=@n9) -> (Output_Get:UI Kit Tab Group Button View Model Object Reference)
   > n10 = Cast To FortVariantSelectorFilterVM!(execute, Object=@n8) -> (then, CastFailed, AsFort Variant Selector Filter VM:Fort Variant Selector Filter VM Object Reference) {is=title K2Node_DynamicCast}
   n11 = Get MainFilter!(self:Fort Variant Selector Filter VM Object Reference=@n10) -> (MainFilter:Fort Variant Selector Sub Filter VM Object Reference)
   > n12 = Set Icon!(
    execute
    Icon:Texture 2D Object Reference=GetIcon!(self, Asset:Soft Object Reference=Get Icon!(self:Fort Variant Selector Sub Filter VM Object Reference=@n11) -> (Icon:Soft Object Reference)) -> (Texture:Texture 2D Object Reference) {is=title K2Node_CallFunction}
    self:UI Kit Selectable Button View Model Object Reference=Get LoopTabVM() -> (LoopTabVM:UI Kit Tab Group Button View Model Object Reference)
   ) -> (then, Output_Get:Texture 2D Object Reference)
   Add(
    execute=[@n14.then, @n12.then]
    TargetArray:Array of UI Kit Tab Group Button View Model Object References=Get TabGroupButtonsViewModels() -> (TabGroupButtonsViewModels:Array of UI Kit Tab Group Button View Model Object References)
    NewItem:"UI Kit Tab Group Button View Model Object Reference (by ref)"=Get LoopTabVM()
   ) {is=title K2Node_CallArrayFunction}
   n13 = Cast To FortVariantSelectorSubFilterVM!(execute=@n10.CastFailed, Object=@n8) -> (then, CastFailed, AsFort Variant Selector Sub Filter VM:Fort Variant Selector Sub Filter VM Object Reference) {is=title K2Node_DynamicCast}
   > n14 = Set Icon(Icon=GetIcon(Asset=Get Icon(self=@n13)), self=Get LoopTabVM())
   n15 = Construct Object from Class(execute=[@n7.then_2, @n2.Is Not Valid], Class=/Script/FortniteUI.UIKitTabGroupViewModel) -> (ReturnValue:UI Kit Tab Group View Model Object Reference)
   > Set Tabs!(
    execute
    Tabs:Array of UI Kit Tab Group Button View Model Object References=Get TabGroupButtonsViewModels()
    self:UI Kit Tab Group View Model Object Reference=@n15
   ) -> (then, Output_Get:Array of UI Kit Tab Group Button View Model Object References)
   > SetUIKitTabGroupViewModel!(execute, self:WBP UIKit Tabs Object Reference=Get TabBase(), Viewmodel:UI Kit Tab Group View Model Object Reference=@n15) -> (then)
  ]}
 TryFocusToEquipped {nodes [
   TryFocusToEquipped() {is=title K2Node_FunctionEntry}
   n1 = Get CommonInputSubsystem(PlayerController=GetOwningPlayer())
   > Branch(Condition="Equal (Enum)"(A=GetCurrentInputType(self=@n1), B=Gamepad))
   n2 = GetBoxEntries()
   > n3 = Branch(Condition=Is Valid Index(TargetArray=@n2))
   > n4 = Cast To WBP_UIKit_VariantRow_Wrapper(Object="Get (a copy)"(Array=@n2))
   > SelectEquippedFocusTarget(self:WBP UIKit Variant Row Wrapper Object Reference=@n4)
   SetFocus(execute=[@n4.CastFailed, @n3.else], self=GetDesiredFocusTarget!(self:Self Object Reference) -> (ReturnValue:Widget Object Reference))
  ]}
 VM_PrimaryFiltersUpdated {nodes [
   n1 = VM_PrimaryFiltersUpdated!() -> (then, FilterVMs:Array of Fort Variant Selector Filter VM Object References) {is=title K2Node_FunctionEntry}
   > n2 = Get Valid Primary Tabs()
   > SetTabgroupFromFilters!(
    execute
    self
    FilterVMs:Array of Object References=@n1
    TabBase:WBP UIKit Tabs Object Reference=@n2
    TabGroupHeader:Common Text Object Reference
    OnTabSelected:"Delegate (by ref)"=Create Event()
   ) -> (then) {is=title K2Node_CallFunction}
   n3 = GetPrimaryFilterTabs_Optional()
   n4 = GetTabIndexByID(self=@n3, TabNameID=0) -> (ReturnValue=0)
   n5 = Get PrimaryFilters(self=Get FortVariantSelectorVM())
   > n6 = Branch(Condition=Is Valid Index(TargetArray:Array of Fort Variant Selector Filter VM Object References=@n5, IndexToTest=@n4))
   > n7 = SetCurrentPrimaryFilter(
    self=Get FortVariantSelectorVM()
    InCurrentPrimaryFilter="Get (a copy)"(Array:Array of Fort Variant Selector Filter VM Object References=@n5, Dimension 1=@n4) -> (Output:Fort Variant Selector Filter VM Object Reference)
   )
   SelectTabByIndex(execute=[@n7.then, @n6.else], self=@n3, Index=0, bSuppressClickFeedback=true)
   n8 = Get PrimaryFilters(self=Get FortVariantSelectorVM())
   n9 = Branch(execute=@n2.NotValid, Condition="Equal (Integer)"(A=Length(TargetArray:Array of Fort Variant Selector Filter VM Object References=@n8), B=1))
   > SetCurrentPrimaryFilter(self=Get FortVariantSelectorVM(), InCurrentPrimaryFilter="Get (a copy)"(Array=@n8))
   PrintString!(
    execute=@n9.else
    InString="WBP_UIKit_VariantSelector_Base - You have more than one primary filter type but you have not specified a tab control for primary filters.  Using the first filter."
    bPrintToScreen
    bPrintToLog
    TextColor
    Duration
    Key
   ) -> (then)
  ]}
 VM_SecondaryFiltersUpdated {nodes [
   n1 = VM_SecondaryFiltersUpdated!() -> (then, SubFilterVMs:Array of Fort Variant Selector Sub Filter VM Object References) {is=title K2Node_FunctionEntry}
   > n2 = Get Valid Secondary Tabs()
   > SetTabgroupFromFilters(
    FilterVMs=@n1
    TabBase=@n2
    TabGroupHeader=GetSecondaryFilterTextBlock_Optional!(self) -> (TextBlock:Common Text Object Reference) {is=title K2Node_CallFunction}
    OnTabSelected=Create Event()
   )
   n3 = GetSecondaryFilterTabs_Optional()
   n4 = GetTabIndexByID(self=@n3, TabNameID=0) -> (ReturnValue=0)
   n5 = Get SecondaryFilters(self=Get FortVariantSelectorVM())
   > n6 = Branch(Condition=Is Valid Index(TargetArray:Array of Fort Variant Selector Sub Filter VM Object References=@n5, IndexToTest=@n4))
   > n7 = SetCurrentSecondaryFilter(
    self=Get FortVariantSelectorVM()
    InCurrentSecondaryFilter="Get (a copy)"(Array:Array of Fort Variant Selector Sub Filter VM Object References=@n5, Dimension 1=@n4) -> (Output:Fort Variant Selector Sub Filter VM Object Reference)
   )
   SelectTabByIndex(execute=[@n7.then, @n6.else], self=@n3, Index=0, bSuppressClickFeedback=true)
  ]}
 VM_SelectedItemUpdated {nodes [
   n1 = VM_SelectedItemUpdated!() -> (then, SelectedItem:Fort Item VM Object Reference) {is=title K2Node_FunctionEntry}
   > n2 = Get()
   > Unbind all Events from On Selected Item Set Event!(execute:Exec, self:Fort Variant Selector VM Object Reference=@n2) -> (then:Exec) {is=title K2Node_ClearDelegate}
   > Bind Event to On Selected Item Set Event(self:Fort Variant Selector VM Object Reference=@n2, Delegate=Create Event()) {is=title K2Node_AddDelegate}
   > Set Selected Item Async!(execute, self:Fort Variant Selector VM Object Reference=@n2, InSelectedItem:Fort Item VM Object Reference=@n1) -> (then) {is=title K2Node_CallFunction}
  ]}
 VM_SetPrimaryFilterText {nodes [
   n1 = VM_SetPrimaryFilterText!() -> (then, PrimaryFilterText:Text) {is=title K2Node_FunctionEntry}
   n2 = GetPrimaryFilterTextBlock_Optional!(self) -> (PrimaryFilterTextBlock.:Common Text Object Reference) {is=title K2Node_CallFunction}
   > Is Valid(InputObject=@n2)
   > "SetText (Text)"(self=@n2, InText=@n1)
  ]}
 VM_SetSecondaryFilterText {nodes [
   n1 = VM_SetSecondaryFilterText!() -> (then, SecondaryFilterText:Text) {is=title K2Node_FunctionEntry}
   n2 = GetSecondaryFilterTextBlock_Optional()
   > Is Valid(InputObject=@n2)
   > "SetText (Text)"(self=@n2, InText=@n1)
  ]}
 VM_VariantSetsUpdated {nodes [
   n1 = VM_VariantSetsUpdated!() -> (then, VariantSetVMs:Array of Fort Variant Set VM Object References) {is=title K2Node_FunctionEntry}
   > Reset!(execute, self:Dynamic Entry Box Object Reference=GetChannelRowsBox_Mandatory() {is=title K2Node_CallFunction}, bDeleteWidgets:bool=true) -> (then)
   > n2 = For Each Loop(Array:Array of Fort Variant Set VM Object References=@n1) -> (Array Element:Fort Variant Set VM Object Reference)
   > n3 = Create Entry!(execute, self:Dynamic Entry Box Object Reference=GetChannelRowsBox_Mandatory()) -> (then, ReturnValue:User Widget Object Reference)
   > n4 = Cast To WBP_UIKit_VariantRow_Wrapper(Object=@n3)
   > Unbind all Events from On Conflicts Updated!(execute:Exec, self:WBP UIKit Variant Row Wrapper Object Reference=@n4) -> (then:Exec) {is=title K2Node_ClearDelegate}
   > Bind Event to On Conflicts Updated(self:WBP UIKit Variant Row Wrapper Object Reference=@n4, Delegate=Create Event()) {is=title K2Node_AddDelegate}
   > Unbind all Events from On Variant Item Engaged!(execute:Exec, self:WBP UIKit Variant Row Wrapper Object Reference=@n4) -> (then:Exec) {is=title K2Node_ClearDelegate}
   > Bind Event to On Variant Item Engaged(self:WBP UIKit Variant Row Wrapper Object Reference=@n4, Delegate=Create Event()) {is=title K2Node_AddDelegate}
   > Unbind all Events from On Variant Item Previewed!(execute:Exec, self:WBP UIKit Variant Row Wrapper Object Reference=@n4) -> (then:Exec) {is=title K2Node_ClearDelegate}
   > Bind Event to On Variant Item Previewed(self:WBP UIKit Variant Row Wrapper Object Reference=@n4, Delegate=Create Event()) {is=title K2Node_AddDelegate}
   > Set Additional Arguments!(
    execute
    Additional Arguments=Get Additional Arguments() -> (Additional Arguments:Map of Names to Booleans)
    self:WBP UIKit Variant Row Wrapper Object Reference=@n4
   ) -> (then, Output_Get)
   > Set TargetStyleTag!(
    execute
    TargetStyleTag:Gameplay Tag Structure=Get VariantChannelStyleTag() -> (VariantChannelStyleTag:Gameplay Tag Structure)
    self:WBP UIKit Variant Row Wrapper Object Reference=@n4
   ) -> (then, Output_Get:Gameplay Tag Structure)
   > Set FortVariantSetVM!(execute, FortVariantSetVM:Fort Variant Set VM Object Reference=@n2, self:WBP UIKit Variant Row Wrapper Object Reference=@n4) -> (then, Output_Get:Fort Variant Set VM Object Reference)
   > Set FortVariantSelectorVM!(
    execute
    FortVariantSelectorVM:Fort Variant Selector VM Object Reference=Get FortVariantSelectorVM()
    self:WBP UIKit Variant Row Wrapper Object Reference=@n4
   ) -> (then, Output_Get:Fort Variant Selector VM Object Reference)
   SetAllRowsNavigation(execute=@n2.Completed) {is=title K2Node_CallFunction}
   > SetActiveWidget(
    self=GetContentSwitcher_Mandatory() {is=title K2Node_CallFunction}
    Widget=GetContentWidget_Mandatory!(self) -> (ContentWidget:Widget Object Reference) {is=title K2Node_CallFunction}
   )
   n5 = Get bIsTabGroupFocused() -> (bIsTabGroupFocused:bool)
   n6 = GetCurrentInputType(self=Get CommonInputSubsystem(PlayerController=GetOwningPlayer()))
   > Branch(Condition=AND Boolean(A="Equal (Enum)"(A=@n6, B=Gamepad), B=NOT Boolean(A=@n5)))
   > TryFocusToEquipped() {is=title K2Node_CallFunction}
  ]}
]
variables [
 Additional Arguments {type="map<boolean>", default {}}
 AllowSelectionOfLockedVariants {type=boolean}
 bIsTabGroupFocused {type=boolean}
 DefaultFilterIcon {type=Texture2D, default=/Game/UI/Foundation/Textures/Icons/Locker/T_Icon_Spray_Adjustments_Tab_128}
 DisplayUnownedVariants {type=boolean}
 ForceApplyVariantOverrides {type=boolean}
 MobileTabIconSize {type=Vector2D, default=<48 48>}
 MobileTabPadding {type=Margin, default=<ltrb 3 3 3 3>}
 OnConflictTextUpdated {}
 OnFilterChanged {}
 OnRotationRequested {}
 OnVariantItemEngaged {}
 OnVariantItemPreviewed {}
 SelectorMode {type=integer}
 ShowAgeGateWarnings {type=boolean}
 ShowUnownedVariantsAsLocked {type=boolean, default=true}
 VariantChannelStyleTag {type=GameplayTag, default {tagName=None}}
 VariantPropertyFilterTag {type=GameplayTag, default {tagName=None}}
]
widgets {info {inheritedWidgetCount=0, parentClass=&UserWidget, rootWidgetClass=null, widgetCount=0}, widgets []}
